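# Control Rig for the TremorMark male medium body, posing the five dyn_jacket bones procedurally and nudging the backpack attach bone.
# After the shared DeformRig function, sequenced branches use spherical pose readers on thighs, pelvis and spine to weight additive bone offsets.
# Crouch, emote, skydive, backpack and rigid-body variables gate or adjust those offsets, mostly smoothed by AlphaInterp.
# Similar branches under Sequence_3 and Sequence_7 are disconnected from execution.
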ControlRigBlueprint /BRCosmetics/Characters/Player/Male/Medium/Bodies/M_MED_TremorMark/Meshes/M_MED_TremorMark_CtrlRig
graphs [
 RigVMFunctionLibrary {links [], nodes []}
 RigVMFunction_Sequence_2_ContainedGraph {nodes [
   n1 = RigVMFunctionEntryNode()
   RigVMFunction_Sequence = &RigVMFunction_Sequence(ExecuteContext=@n1.ExecuteContext)
   n2 = &RigVMFunction_Sequence(ExecuteContext=@RigVMFunction_Sequence.B)
   n3 = &RigVMFunction_Sequence(ExecuteContext=@n2.B)
   RigVMFunctionReturnNode(A=@RigVMFunction_Sequence.A, B=@n2.A, C=@n3.A, D=@n3.B)
  ]}
 RigVMFunction_Sequence_2_ContainedGraph {nodes [
   n1 = RigVMFunctionEntryNode()
   n2 = &RigVMFunction_Sequence(ExecuteContext=@n1.ExecuteContext)
   n3 = &RigVMFunction_Sequence(ExecuteContext=@n2.B)
   RigVMFunctionReturnNode(A=@n2.A, B=@n3.A, C=@n3.B)
  ]}
 RigVMFunction_Sequence_2_ContainedGraph {nodes [
   n1 = RigVMFunctionEntryNode()
   RigVMFunction_Sequence = &RigVMFunction_Sequence(ExecuteContext=@n1.ExecuteContext)
   n2 = &RigVMFunction_Sequence(ExecuteContext=@RigVMFunction_Sequence.B)
   n3 = &RigVMFunction_Sequence(ExecuteContext=@n2.B)
   RigVMFunctionReturnNode(A=@RigVMFunction_Sequence.A, B=@n2.A, C=@n3.A, D=@n3.B)
  ]}
 RigVMFunction_Sequence_7_ContainedGraph {nodes [
   n1 = RigVMFunctionEntryNode()
   n2 = &RigVMFunction_Sequence(ExecuteContext=@n1.ExecuteContext)
   n3 = &RigVMFunction_Sequence(ExecuteContext=@n2.B)
   n4 = &RigVMFunction_Sequence(ExecuteContext=@n3.B)
   RigVMFunctionReturnNode(B=@n3.A, C=@n4.A, D=@n4.B, A=@n2.A)
  ]}
 RigVMFunction_Sequence_7_ContainedGraph {nodes [
   n1 = RigVMFunctionEntryNode()
   n2 = &RigVMFunction_Sequence(ExecuteContext=@n1.ExecuteContext)
   n3 = &RigVMFunction_Sequence(ExecuteContext=@n2.B)
   RigVMFunctionReturnNode(B=@n3.A, C=@n3.B, A=@n2.A)
  ]}
 RigVMModel {nodes [
   RigUnit_BeginExecution = &RigUnit_BeginExecution()
   FN_DeformRig_FNC = RigVMFunctionReferenceNode(
    DeformRig_ Curve="(deformRig_l_upperarm_up_alpha_72_#1=False,deformRig_r_upperarm_up_alpha_73_#2=False,deformRig_l_upperarm_fr_alpha_74_#3=False,deformRig_r_upperarm_fr_alpha_75_CB51B8 ...(truncated)"
    DeformRig_Bool="(upperarm_1_#4=True,elbow_3_#5=True,wrist_7_#6=True,groin_8_#7=True,glute_17_#8=True,knee_9_40C408324CA261 ...(truncated)"
    DeformRig_Debug=(upperarm_debug_18_#4=False,wrist_debug_20_#6=False,lat_debug_26_#9=False,groinGlute_debug_27_#7=False)
    DeformRig_Value="(upperarm_up_18_#10=(Rotation=(X=0,Y=0,Z=0,W=1),Translation=(X=0,Y=0,Z=8),Scale3D=(X=0.7,Y=1,Z=1)),upperarm_fwd_fr_19_#11=(Rotation=(X=0,Y=0,Z=0,W=1),Translation=(X=0,Y=10,Z=0),Scale3D=(X=0.5,Y=1,Z= ...(truncated)"
    ExecuteContext=@RigUnit_BeginExecution.ExecutePin
   )
   > RigVMFunction_Sequence_1 = RigVMAggregateNode()
   n1 = &RigUnit_SphericalPoseReader(
    ActiveRegionScaleFactors=(PositiveWidth=1,NegativeWidth=1,PositiveHeight=1,NegativeHeight=1)
    ActiveRegionSize=0.1
    Debug=(bDrawDebug=false,bDraw2D=False,bDrawLocalAxes=False,DebugScale=25,DebugSegments=20,DebugThickness=0.25)
    DriverAxis=(X=-1,Y=0,Z=0)
    DriverItem=(Type=Bone,Name="thigh_l")
    FalloffRegionScaleFactors=(PositiveWidth=1,NegativeWidth=1,PositiveHeight=1,NegativeHeight=1)
    FalloffSize=0.6
    FlipHeightScaling=False
    FlipWidthScaling=False
    OptionalParentItem=(Type=Bone,Name="None")
    RotationOffset=(X=120,Y=0,Z=-30)
    ExecutePin=@RigVMFunction_Sequence_1.A
   )
   n2 = RigVMVariableNode() {variable=is_crouch}
   n3 = &RigVMFunction_AlphaInterp(InterpSpeedDecreasing=50, InterpSpeedIncreasing=500, Scale=1, bClampResult=false, Value=@n1.OutputParam)
   Subtract = &RigVMFunction_MathDoubleSub(B=RigVMVariableNode() {variable=is_emote}.Value, A=@n3.Result)
   Clamp = &RigVMFunction_MathDoubleClamp(Value=@Subtract.Result)
   n4 = &RigVMFunction_AlphaInterp(InterpSpeedDecreasing=50, InterpSpeedIncreasing=750, Scale=1, bClampResult=false, bInterpResult=false, Value=@n2.Value)
   n5 = &RigVMFunction_MathDoubleMul(A=45, B=@n4.Result)
   n6 = &RigVMFunction_MathDoubleMul(A=20, B=@n4.Result)
   n7 = &RigVMFunction_MathDoubleMul(A=10, B=@n4.Result)
   > &RigUnit_ModifyTransforms(
    ItemToModify=((Item=(Type=Bone,Name="dyn_jacket_1_base"),Transform=(Rotation=(X=0,Y=0,Z=0,W=1),Translation=(X=-3,Y=0,Z=5),Scale3D=(X=1,Y=1,Z=1))))
    WeightMinimum=0
    ItemToModify.0.Transform.Translation.X=&RigVMFunction_MathDoubleAdd(A=-3, B=&RigVMFunction_MathDoubleMul(A=-3, B=@n2.Value).Result).Result
    ItemToModify.0.Transform.Translation.Y=&RigVMFunction_MathDoubleAdd(A=0, B=&RigVMFunction_MathDoubleMul(A=2, B=@n2.Value).Result).Result
    ItemToModify.0.Transform.Translation.Z=&RigVMFunction_MathDoubleAdd(A=2, B=&RigVMFunction_MathDoubleMul(A=0, B=@n2.Value).Result).Result
    Weight=@Clamp.Result
    ItemToModify.0.Transform.Rotation=&RigVMFunction_MathQuaternionFromEuler(
     Euler=(X=0,Y=0,Z=0)
     RotationOrder=XYZ
     Euler.X=&RigVMFunction_MathDoubleAdd(A=45, B=@n5.Result).Result
     Euler.Y=&RigVMFunction_MathDoubleAdd(A=-30, B=@n6.Result).Result
     Euler.Z=&RigVMFunction_MathDoubleAdd(A=-10, B=@n7.Result).Result
    ).Result
   )
   n8 = RigVMVariableNode() {variable=is_crouch}
   n9 = &RigVMFunction_MathDoubleMul(A=10, B=@n8.Value)
   n10 = &RigVMFunction_MathDoubleMul(A=0, B=@n8.Value)
   n11 = &RigVMFunction_MathDoubleMul(A=0, B=@n8.Value)
   > &RigUnit_ModifyTransforms(
    ItemToModify=((Item=(Type=Bone,Name="dyn_jacket_2_base"),Transform=(Rotation=(X=0.0872,Y=0,Z=0,W=0.9962),Translation=(X=-3,Y=0,Z=0),Scale3D=(X=1,Y=1,Z=1))))
    WeightMinimum=0
    ItemToModify.0.Transform.Translation.X=&RigVMFunction_MathDoubleAdd(A=-2, B=&RigVMFunction_MathDoubleMul(A=-3, B=@n8.Value).Result).Result
    ItemToModify.0.Transform.Translation.Y=&RigVMFunction_MathDoubleAdd(A=2, B=&RigVMFunction_MathDoubleMul(A=-2, B=@n8.Value).Result).Result
    ItemToModify.0.Transform.Translation.Z=&RigVMFunction_MathDoubleAdd(A=-5, B=&RigVMFunction_MathDoubleMul(A=5, B=@n8.Value).Result).Result
    ItemToModify.0.Transform.Rotation=&RigVMFunction_MathQuaternionFromEuler(
     Euler=(X=0,Y=0,Z=0)
     RotationOrder=XYZ
     Euler.X=&RigVMFunction_MathDoubleAdd(A=0, B=@n9.Result).Result
     Euler.Y=&RigVMFunction_MathDoubleAdd(A=-20, B=@n10.Result).Result
     Euler.Z=&RigVMFunction_MathDoubleAdd(A=-10, B=@n11.Result).Result
    ).Result
    Weight=@n3.Result
   )
   n12 = &RigVMFunction_AlphaInterp(InterpSpeedDecreasing=10, InterpSpeedIncreasing=10, Scale=1, bClampResult=False, Value=@n8.Value)
   > &RigUnit_ModifyTransforms(
    ItemToModify=((Item=(Type=Bone,Name="dyn_jacket_3_base"),Transform=(Rotation=(X=0,Y=0,Z=0,W=1),Translation=(X=0,Y=-1,Z=0),Scale3D=(X=1,Y=1,Z=1))))
    WeightMinimum=0
    Weight=@n12.Result
   )
   n13 = &RigUnit_SphericalPoseReader(
    ActiveRegionScaleFactors=(PositiveWidth=1,NegativeWidth=1,PositiveHeight=1,NegativeHeight=1)
    ActiveRegionSize=0.1
    Debug=(bDrawDebug=true,bDraw2D=False,bDrawLocalAxes=False,DebugScale=25,DebugSegments=20,DebugThickness=0.25)
    DriverAxis=(X=1,Y=0,Z=0)
    DriverItem=(Type=Bone,Name="thigh_r")
    FalloffRegionScaleFactors=(PositiveWidth=1,NegativeWidth=1,PositiveHeight=1,NegativeHeight=1)
    FalloffSize=0.5
    FlipHeightScaling=False
    FlipWidthScaling=False
    OptionalParentItem=(Type=Bone,Name="None")
    RotationOffset=(X=-75,Y=0,Z=-25)
    ExecutePin=@RigVMFunction_Sequence_1.B
   )
   n14 = RigVMVariableNode() {variable=is_crouch}
   n15 = &RigVMFunction_MathDoubleMul(A=-15, B=@n14.Value)
   n16 = &RigVMFunction_MathDoubleMul(A=0, B=@n14.Value)
   n17 = &RigVMFunction_MathDoubleMul(A=-20, B=@n14.Value)
   n18 = &RigVMFunction_AlphaInterp(InterpSpeedDecreasing=1, InterpSpeedIncreasing=5, Scale=1, bClampResult=False, Value=@n13.OutputParam)
   > &RigUnit_ModifyTransforms(
    ItemToModify=((Item=(Type=Bone,Name="dyn_jacket_5_base"),Transform=(Rotation=(X=-0.6744,Y=0.2126,Z=0.0308,W=0.7064),Translation=(X=-5,Y=0,Z=-5),Scale3D=(X=1,Y=1,Z=1))))
    WeightMinimum=0
    ItemToModify.0.Transform.Translation.X=&RigVMFunction_MathDoubleAdd(A=-2, B=&RigVMFunction_MathDoubleMul(A=-1.5, B=@n14.Value).Result).Result
    ItemToModify.0.Transform.Translation.Y=&RigVMFunction_MathDoubleAdd(A=0, B=&RigVMFunction_MathDoubleMul(A=0, B=@n14.Value).Result).Result
    ItemToModify.0.Transform.Translation.Z=&RigVMFunction_MathDoubleAdd(A=-3, B=&RigVMFunction_MathDoubleMul(A=0, B=@n14.Value).Result).Result
    ItemToModify.0.Transform.Rotation=&RigVMFunction_MathQuaternionFromEuler(
     Euler=(X=0,Y=0,Z=0)
     Euler.X=&RigVMFunction_MathDoubleAdd(A=-30, B=@n15.Result).Result
     Euler.Y=&RigVMFunction_MathDoubleAdd(A=10, B=@n16.Result).Result
     Euler.Z=&RigVMFunction_MathDoubleAdd(A=0, B=@n17.Result).Result
    ).Result
    Weight=@n18.Result
   )
   n19 = RigVMVariableNode() {variable=is_crouch}
   n20 = &RigVMFunction_MathDoubleMul(A=-30, B=@n19.Value)
   n21 = &RigVMFunction_MathDoubleMul(A=0, B=@n19.Value)
   n22 = &RigVMFunction_MathDoubleMul(A=0, B=@n19.Value)
   > &RigUnit_ModifyTransforms(
    ItemToModify=((Item=(Type=Bone,Name="dyn_jacket_4_base"),Transform=(Rotation=(X=-0.13,Y=-0.0864,Z=-0.0114,W=0.9877),Translation=(X=-2,Y=0,Z=0),Scale3D=(X=1,Y=1,Z=1))))
    WeightMinimum=0
    ItemToModify.0.Transform.Translation.X=&RigVMFunction_MathDoubleAdd(A=-2, B=&RigVMFunction_MathDoubleMul(A=-1, B=@n19.Value).Result).Result
    ItemToModify.0.Transform.Translation.Y=&RigVMFunction_MathDoubleAdd(A=0, B=&RigVMFunction_MathDoubleMul(A=0, B=@n19.Value).Result).Result
    ItemToModify.0.Transform.Translation.Z=&RigVMFunction_MathDoubleAdd(A=0, B=&RigVMFunction_MathDoubleMul(A=0, B=@n19.Value).Result).Result
    ItemToModify.0.Transform.Rotation=&RigVMFunction_MathQuaternionFromEuler(
     Euler=(X=0,Y=0,Z=0)
     Euler.X=&RigVMFunction_MathDoubleAdd(A=0, B=@n20.Result).Result
     Euler.Y=&RigVMFunction_MathDoubleAdd(A=0, B=@n21.Result).Result
     Euler.Z=&RigVMFunction_MathDoubleAdd(A=0, B=@n22.Result).Result
    ).Result
    Weight=@n18.Result
   )
   n23 = &RigUnit_SphericalPoseReader(
    ActiveRegionScaleFactors=(PositiveWidth=0.5,NegativeWidth=0.5,PositiveHeight=0.5,NegativeHeight=0.5)
    ActiveRegionSize=0.1
    Debug=(bDrawDebug=false,bDraw2D=False,bDrawLocalAxes=False,DebugScale=25,DebugSegments=20,DebugThickness=0.25)
    DriverAxis=(X=0,Y=-1,Z=0)
    DriverItem=(Type=Bone,Name="pelvis")
    FalloffRegionScaleFactors=(PositiveWidth=1,NegativeWidth=1,PositiveHeight=1,NegativeHeight=1)
    FalloffSize=0.25
    FlipHeightScaling=False
    FlipWidthScaling=False
    OptionalParentItem=(Type=Bone,Name="None")
    RotationOffset=(X=-145,Y=0,Z=15)
    ExecutePin=@RigVMFunction_Sequence_1.C
   )
   n24 = RigVMVariableNode() {variable=is_crouch}
   n25 = &RigVMFunction_MathDoubleMul(A=0, B=@n24.Value)
   n26 = &RigVMFunction_MathDoubleMul(A=0, B=@n24.Value)
   n27 = &RigVMFunction_MathDoubleMul(A=-20, B=@n24.Value)
   > &RigUnit_ModifyTransforms(
    ItemToModify=((Item=(Type=Bone,Name="dyn_jacket_1_base"),Transform=(Rotation=(X=0,Y=0,Z=0,W=1),Translation=(X=-3,Y=0,Z=5),Scale3D=(X=1,Y=1,Z=1))))
    WeightMinimum=0
    ItemToModify.0.Transform.Translation.X=&RigVMFunction_MathDoubleAdd(A=0, B=&RigVMFunction_MathDoubleMul(A=2.5, B=@n24.Value).Result).Result
    ItemToModify.0.Transform.Translation.Y=&RigVMFunction_MathDoubleAdd(A=0, B=&RigVMFunction_MathDoubleMul(A=-2.5, B=@n24.Value).Result).Result
    ItemToModify.0.Transform.Translation.Z=&RigVMFunction_MathDoubleAdd(A=0, B=&RigVMFunction_MathDoubleMul(A=2.5, B=@n24.Value).Result).Result
    ItemToModify.0.Transform.Rotation=&RigVMFunction_MathQuaternionFromEuler(
     Euler=(X=0,Y=0,Z=0)
     RotationOrder=XYZ
     Euler.X=&RigVMFunction_MathDoubleAdd(A=0, B=@n25.Result).Result
     Euler.Y=&RigVMFunction_MathDoubleAdd(A=0, B=@n26.Result).Result
     Euler.Z=&RigVMFunction_MathDoubleAdd(A=0, B=@n27.Result).Result
    ).Result
    Weight=@n23.OutputParam
   )
   > &RigUnit_ModifyTransforms(
    ItemToModify=((Item=(Type=Bone,Name="dyn_jacket_2_base"),Transform=(Rotation=(X=0.2549,Y=-0.0449,Z=-0.1677,W=0.9513),Translation=(X=-3,Y=0,Z=5),Scale3D=(X=1,Y=1,Z=1))))
    WeightMinimum=0
    ItemToModify.0.Transform.Translation.X=&RigVMFunction_MathDoubleAdd(A=0, B=&RigVMFunction_MathDoubleMul(A=0, B=@n24.Value).Result).Result
    ItemToModify.0.Transform.Translation.Y=&RigVMFunction_MathDoubleAdd(A=0, B=&RigVMFunction_MathDoubleMul(A=2, B=@n24.Value).Result).Result
    ItemToModify.0.Transform.Translation.Z=&RigVMFunction_MathDoubleAdd(A=0, B=&RigVMFunction_MathDoubleMul(A=2, B=@n24.Value).Result).Result
    Weight=@n23.OutputParam
   )
   > &RigUnit_ModifyTransforms(
    ItemToModify=((Item=(Type=Bone,Name="dyn_jacket_3_base"),Transform=(Rotation=(X=0,Y=-0.1736,Z=0,W=0.9848),Translation=(X=0,Y=0,Z=0),Scale3D=(X=1,Y=1,Z=1))))
    WeightMinimum=0
    Weight=&RigVMFunction_MathDoubleMul(B=@n24.Value, A=@n23.OutputParam).Result
   )
   RigVMFunction_Sequence_6 = RigVMAggregateNode(ExecuteContext=@RigVMFunction_Sequence_1.D)
   n28 = &RigVMFunction_Sequence(ExecuteContext=@RigVMFunction_Sequence_6.A)
   n29 = &RigUnit_SphericalPoseReader(
    ActiveRegionScaleFactors=(PositiveWidth=1,NegativeWidth=1,PositiveHeight=1,NegativeHeight=1)
    ActiveRegionSize=0.1
    Debug=(bDrawDebug=false,bDraw2D=False,bDrawLocalAxes=False,DebugScale=25,DebugSegments=20,DebugThickness=0.25)
    DriverAxis=(X=0,Y=1,Z=0)
    DriverItem=(Type=Bone,Name="spine_03")
    FalloffRegionScaleFactors=(PositiveWidth=0.25,NegativeWidth=0.5,PositiveHeight=0.5,NegativeHeight=0.5)
    FalloffSize=0.35
    FlipHeightScaling=False
    FlipWidthScaling=False
    OptionalParentItem=(Type=Bone,Name="None")
    RotationOffset=(X=90,Y=0,Z=35)
    ExecutePin=@n28.A
   )
   n30 = &RigVMFunction_AlphaInterp(InterpSpeedDecreasing=20, InterpSpeedIncreasing=20, Scale=1, bClampResult=False, Value=@n29.OutputParam)
   n31 = &RigVMFunction_AlphaInterp(
    InterpSpeedDecreasing=10
    InterpSpeedIncreasing=10
    Scale=1
    bClampResult=False
    Value=&RigVMFunction_MathDoubleMul(B=3, A=RigVMVariableNode() {variable=is_emote}.Value).Result
   )
   > &RigUnit_ModifyTransforms(
    ItemToModify=((Item=(Type=Bone,Name="dyn_jacket_1_base"),Transform=(Rotation=(X=0,Y=0,Z=0,W=1),Translation=(X=-2,Y=5,Z=4),Scale3D=(X=1,Y=1,Z=1))))
    WeightMinimum=0
    Weight=&RigVMFunction_MathDoubleMul(A=@n30.Result, B=@n31.Result).Result
    ItemToModify.0.Transform.Rotation=&RigVMFunction_MathQuaternionFromEuler(Euler=(X=30,Y=-10,Z=15)).Result
   )
   > &RigUnit_ModifyTransforms(
    ItemToModify=((Item=(Type=Bone,Name="dyn_jacket_2_base"),Transform=(Rotation=(X=-0.13,Y=-0.0864,Z=-0.0114,W=0.9877),Translation=(X=-1,Y=0,Z=1),Scale3D=(X=1,Y=1,Z=1))))
    WeightMinimum=0
    ItemToModify.0.Transform.Rotation=&RigVMFunction_MathQuaternionFromEuler(Euler=(X=0,Y=0,Z=0)).Result
    Weight=@n31.Result
   )
   > &RigUnit_ModifyTransforms(
    ItemToModify=((Item=(Type=Bone,Name="dyn_jacket_5_base"),Transform=(Rotation=(X=-0.6744,Y=0.2126,Z=0.0308,W=0.7064),Translation=(X=0,Y=2,Z=-2),Scale3D=(X=1,Y=1,Z=1))))
    WeightMinimum=0
    ItemToModify.0.Transform.Rotation=&RigVMFunction_MathQuaternionFromEuler(Euler=(X=0,Y=0,Z=0)).Result
    Weight=@n31.Result
   )
   > &RigUnit_ModifyTransforms(
    ItemToModify=((Item=(Type=Bone,Name="dyn_jacket_4_base"),Transform=(Rotation=(X=-0.13,Y=-0.0864,Z=-0.0114,W=0.9877),Translation=(X=0,Y=0,Z=-1),Scale3D=(X=1,Y=1,Z=1))))
    WeightMinimum=0
    ItemToModify.0.Transform.Rotation=&RigVMFunction_MathQuaternionFromEuler(Euler=(X=0,Y=0,Z=0)).Result
    Weight=@n31.Result
   )
   n32 = &RigUnit_SphericalPoseReader(
    ActiveRegionScaleFactors=(PositiveWidth=0.5,NegativeWidth=0.5,PositiveHeight=1,NegativeHeight=1)
    ActiveRegionSize=0.1
    Debug=(bDrawDebug=true,bDraw2D=False,bDrawLocalAxes=False,DebugScale=25,DebugSegments=20,DebugThickness=0.25)
    DriverAxis=(X=1,Y=0,Z=0)
    DriverItem=(Type=Bone,Name="thigh_r")
    FalloffRegionScaleFactors=(PositiveWidth=1,NegativeWidth=1,PositiveHeight=1,NegativeHeight=1)
    FalloffSize=0.25
    FlipHeightScaling=False
    FlipWidthScaling=False
    OptionalParentItem=(Type=Bone,Name="None")
    RotationOffset=(X=-45,Y=-20,Z=0)
    ExecutePin=@n28.B
   )
   n33 = &RigVMFunction_AlphaInterp(InterpSpeedDecreasing=20, InterpSpeedIncreasing=20, Scale=1, bClampResult=False, Value=@n32.OutputParam)
   n34 = &RigVMFunction_MathDoubleMul(B=@n33.Result, A=@n31.Result)
   > &RigUnit_ModifyTransforms(
    ItemToModify=((Item=(Type=Bone,Name="dyn_jacket_1_base"),Transform=(Rotation=(X=0,Y=0,Z=0,W=1),Translation=(X=-1,Y=1,Z=1),Scale3D=(X=1,Y=1,Z=1))))
    WeightMinimum=0
    ItemToModify.0.Transform.Rotation=&RigVMFunction_MathQuaternionFromEuler(Euler=(X=0,Y=-10,Z=10)).Result
    Weight=@n34.Result
   )
   > &RigUnit_ModifyTransforms(
    ItemToModify=((Item=(Type=Bone,Name="dyn_jacket_2_base"),Transform=(Rotation=(X=0,Y=0,Z=0,W=1),Translation=(X=-1,Y=-1,Z=1),Scale3D=(X=1,Y=1,Z=1))))
    WeightMinimum=0
    ItemToModify.0.Transform.Rotation=&RigVMFunction_MathQuaternionFromEuler(Euler=(X=0,Y=-10,Z=-10)).Result
    Weight=@n34.Result
   )
   > &RigUnit_ModifyTransforms(
    ItemToModify=((Item=(Type=Bone,Name="dyn_jacket_3_base"),Transform=(Rotation=(X=0,Y=0,Z=0,W=1),Translation=(X=-2,Y=-2,Z=-1),Scale3D=(X=1,Y=1,Z=1))))
    WeightMinimum=0
    ItemToModify.0.Transform.Rotation=&RigVMFunction_MathQuaternionFromEuler(Euler=(X=0,Y=30,Z=-30)).Result
    Weight=@n34.Result
   )
   > &RigUnit_ModifyTransforms(
    ItemToModify=((Item=(Type=Bone,Name="dyn_jacket_4_base"),Transform=(Rotation=(X=0,Y=0,Z=0,W=1),Translation=(X=-2,Y=-2,Z=-1),Scale3D=(X=1,Y=1,Z=1))))
    WeightMinimum=0
    ItemToModify.0.Transform.Rotation=&RigVMFunction_MathQuaternionFromEuler(Euler=(X=-10,Y=40,Z=20)).Result
    Weight=@n34.Result
   )
   > &RigUnit_ModifyTransforms(
    ItemToModify=((Item=(Type=Bone,Name="dyn_jacket_5_base"),Transform=(Rotation=(X=0,Y=0,Z=0,W=1),Translation=(X=-4,Y=0,Z=-2),Scale3D=(X=1,Y=1,Z=1))))
    WeightMinimum=0
    ItemToModify.0.Transform.Rotation=&RigVMFunction_MathQuaternionFromEuler(Euler=(X=-30,Y=20,Z=30)).Result
    Weight=@n34.Result
   )
   RigVMFunction_Sequence = &RigVMFunction_Sequence(ExecuteContext=@RigVMFunction_Sequence_6.B)
   n35 = &RigUnit_SphericalPoseReader(
    ActiveRegionScaleFactors=(PositiveWidth=2,NegativeWidth=1,PositiveHeight=2,NegativeHeight=2)
    ActiveRegionSize=0.1
    Debug=(bDrawDebug=false,bDraw2D=False,bDrawLocalAxes=False,DebugScale=25,DebugSegments=20,DebugThickness=0.25)
    DriverAxis=(X=0,Y=-1,Z=0)
    DriverItem=(Type=Bone,Name="pelvis")
    FalloffRegionScaleFactors=(PositiveWidth=1,NegativeWidth=1,PositiveHeight=1,NegativeHeight=1)
    FalloffSize=0.5
    FlipHeightScaling=False
    FlipWidthScaling=False
    OptionalParentItem=(Type=Bone,Name="None")
    RotationOffset=(X=-90,Y=0,Z=60)
    ExecutePin=@RigVMFunction_Sequence.A
   )
   n36 = &RigVMFunction_AlphaInterp(
    InterpSpeedDecreasing=50
    InterpSpeedIncreasing=300
    Scale=1
    bClampResult=false
    Value=&RigVMFunction_MathDoubleMul(A=@n35.OutputParam, B=RigVMVariableNode() {variable=is_sky_diving}.Value).Result
   )
   > &RigUnit_ModifyTransforms(
    ItemToModify=((Item=(Type=Bone,Name="dyn_jacket_1_base"),Transform=(Rotation=(X=0.399,Y=-0.0925,Z=0.2234,W=0.8845),Translation=(X=-1.5,Y=6,Z=10),Scale3D=(X=1,Y=1,Z=1))))
    WeightMinimum=0
    Weight=@n36.Result
   )
   > &RigUnit_ModifyTransforms(
    ItemToModify=((Item=(Type=Bone,Name="dyn_jacket_2_base"),Transform=(Rotation=(X=-0.1499,Y=-0.1539,Z=-0.1499,W=0.9651),Translation=(X=-1.5,Y=5,Z=-5),Scale3D=(X=1,Y=1,Z=1))))
    WeightMinimum=0
    Weight=@n36.Result
   )
   > &RigUnit_ModifyTransforms(
    ItemToModify=((Item=(Type=Bone,Name="dyn_jacket_3_base"),Transform=(Rotation=(X=0.0858,Y=-0.0151,Z=-0.173,W=0.9811),Translation=(X=-5,Y=-4,Z=2.5),Scale3D=(X=1,Y=1,Z=1))))
    WeightMinimum=0
    Weight=@n36.Result
   )
   > &RigUnit_ModifyTransforms(
    ItemToModify=((Item=(Type=Bone,Name="dyn_jacket_4_base"),Transform=(Rotation=(X=0.1935,Y=0.4307,Z=-0.2121,W=0.8556),Translation=(X=0,Y=-3,Z=-2),Scale3D=(X=1,Y=1,Z=1))))
    WeightMinimum=0
    Weight=@n36.Result
   )
   > &RigUnit_ModifyTransforms(
    ItemToModify=((Item=(Type=Bone,Name="dyn_jacket_5_base"),Transform=(Rotation=(X=-0.5012,Y=0.2683,Z=0.2941,W=0.7683),Translation=(X=-1,Y=2,Z=-3),Scale3D=(X=1,Y=1,Z=1))))
    WeightMinimum=0
    Weight=@n36.Result
   )
   n37 = &RigVMFunction_AlphaInterp(InterpSpeedDecreasing=10, InterpSpeedIncreasing=10, Scale=1, bClampResult=False, Value=RigVMVariableNode() {variable=is_move_backpack}.Value)
   &RigUnit_ModifyTransforms(
    ItemToModify=((Item=(Type=Bone,Name="attach_backpack"),Transform=(Rotation=(X=-0.1305,Y=0,Z=0,W=0.9914),Translation=(X=0,Y=-1.5,Z=0),Scale3D=(X=1,Y=1,Z=1))))
    WeightMinimum=0
    Weight=@n37.Result
    ExecutePin=@RigVMFunction_Sequence.B
   )
   n38 = &RigUnit_SphericalPoseReader(
    ActiveRegionScaleFactors=(PositiveWidth=0.25,NegativeWidth=0.25,PositiveHeight=0.25,NegativeHeight=0.25)
    ActiveRegionSize=0.1
    Debug=(bDrawDebug=false,bDraw2D=False,bDrawLocalAxes=False,DebugScale=25,DebugSegments=20,DebugThickness=0.25)
    DriverAxis=(X=0,Y=-1,Z=0)
    DriverItem=(Type=Bone,Name="spine_04")
    FalloffRegionScaleFactors=(PositiveWidth=1,NegativeWidth=1,PositiveHeight=1,NegativeHeight=1)
    FalloffSize=0.1
    FlipHeightScaling=False
    FlipWidthScaling=False
    OptionalParentItem=(Type=Bone,Name="None")
    RotationOffset=(X=-115,Y=0,Z=-5)
    ExecutePin=@RigVMFunction_Sequence_6.C
   )
   n39 = &RigVMFunction_AlphaInterp(
    InterpSpeedDecreasing=10
    InterpSpeedIncreasing=10
    Scale=1
    bClampResult=False
    Value=&RigVMFunction_MathDoubleMul(A=@n38.OutputParam, B=RigVMVariableNode() {variable=is_crouch}.Value).Result
   )
   > &RigUnit_ModifyTransforms(
    ItemToModify=((Item=(Type=Bone,Name="dyn_jacket_1_base"),Transform=(Rotation=(X=0,Y=0,Z=0,W=1),Translation=(X=0,Y=0,Z=0),Scale3D=(X=1,Y=1,Z=1))))
    WeightMinimum=0
    ItemToModify.0.Transform.Rotation=&RigVMFunction_MathQuaternionFromEuler(Euler=(X=15,Y=-30,Z=20), RotationOrder=XYZ).Result
    Weight=@n39.Result
   )
   > &RigUnit_ModifyTransforms(
    ItemToModify=((Item=(Type=Bone,Name="dyn_jacket_2_base"),Transform=(Rotation=(X=0,Y=0,Z=0,W=1),Translation=(X=0,Y=0,Z=0),Scale3D=(X=1,Y=1,Z=1))))
    WeightMinimum=0
    ItemToModify.0.Transform.Rotation=&RigVMFunction_MathQuaternionFromEuler(Euler=(X=20,Y=0,Z=-15), RotationOrder=XYZ).Result
    Weight=@n39.Result
   )
   n40 = &RigVMFunction_AlphaInterp(
    InterpSpeedDecreasing=10
    InterpSpeedIncreasing=10
    Scale=1
    bClampResult=False
    Value=RigVMVariableNode() {variable=is_RigidBody_alpha_FN}.Value
   )
   &RigUnit_ModifyTransforms(
    ItemToModify=((Item=(Type=Bone,Name="dyn_jacket_1_base"),Transform=(Rotation=(X=0,Y=0,Z=0,W=1),Translation=(X=0,Y=1,Z=-1),Scale3D=(X=1,Y=1,Z=1))))
    WeightMinimum=0
    Weight=@n40.Result
    ItemToModify.0.Transform.Rotation=&RigVMFunction_MathQuaternionFromEuler(Euler=(X=0,Y=15,Z=0), RotationOrder=XYZ).Result
    ExecutePin=@RigVMFunction_Sequence_6.D
   )
   > &RigUnit_ModifyTransforms(
    ItemToModify=((Item=(Type=Bone,Name="dyn_jacket_5_base"),Transform=(Rotation=(X=0,Y=0,Z=0,W=1),Translation=(X=0,Y=0,Z=0),Scale3D=(X=1,Y=1,Z=1))))
    WeightMinimum=0
    Weight=@n40.Result
    ItemToModify.0.Transform.Rotation=&RigVMFunction_MathQuaternionFromEuler(Euler=(X=0,Y=-15,Z=0), RotationOrder=XYZ).Result
   )
   RigVMFunction_Sequence_3 = RigVMAggregateNode(ExecuteContext=())
   RigVMFunction_Sequence_4 = RigVMAggregateNode(ExecuteContext=@RigVMFunction_Sequence_3.A)
   n41 = &RigUnit_SphericalPoseReader(
    ActiveRegionScaleFactors=(PositiveWidth=0.5,NegativeWidth=0.5,PositiveHeight=0.5,NegativeHeight=0.5)
    ActiveRegionSize=0.1
    Debug=(bDrawDebug=false,bDraw2D=False,bDrawLocalAxes=False,DebugScale=25,DebugSegments=20,DebugThickness=0.25)
    DriverAxis=(X=0,Y=-1,Z=0)
    DriverItem=(Type=Bone,Name="pelvis")
    FalloffRegionScaleFactors=(PositiveWidth=1,NegativeWidth=1,PositiveHeight=1,NegativeHeight=1)
    FalloffSize=0.25
    FlipHeightScaling=False
    FlipWidthScaling=False
    OptionalParentItem=(Type=Bone,Name="None")
    RotationOffset=(X=-145,Y=0,Z=15)
    ExecutePin=@RigVMFunction_Sequence_4.B
   )
   n42 = RigVMVariableNode() {variable=is_crouch}
   n43 = &RigVMFunction_MathDoubleMul(A=-30, B=@n42.Value)
   n44 = &RigVMFunction_MathDoubleMul(A=40, B=0)
   n45 = &RigVMFunction_MathDoubleMul(A=0, B=@n42.Value)
   > &RigUnit_ModifyTransforms(
    ItemToModify=((Item=(Type=Bone,Name="dyn_jacket_1_base"),Transform=(Rotation=(X=0,Y=0,Z=0,W=1),Translation=(X=-3,Y=0,Z=5),Scale3D=(X=1,Y=1,Z=1))))
    WeightMinimum=0
    ItemToModify.0.Transform.Translation.X=&RigVMFunction_MathDoubleAdd(A=0, B=&RigVMFunction_MathDoubleMul(A=-3.5, B=@n42.Value).Result).Result
    ItemToModify.0.Transform.Translation.Y=&RigVMFunction_MathDoubleAdd(A=0, B=&RigVMFunction_MathDoubleMul(A=0, B=@n42.Value).Result).Result
    ItemToModify.0.Transform.Translation.Z=&RigVMFunction_MathDoubleAdd(A=0, B=&RigVMFunction_MathDoubleMul(A=0, B=@n42.Value).Result).Result
    ItemToModify.0.Transform.Rotation=&RigVMFunction_MathQuaternionFromEuler(
     Euler=(X=0,Y=0,Z=0)
     RotationOrder=XYZ
     Euler.X=&RigVMFunction_MathDoubleAdd(A=0, B=@n43.Result).Result
     Euler.Y=&RigVMFunction_MathDoubleAdd(A=-20, B=@n44.Result).Result
     Euler.Z=&RigVMFunction_MathDoubleAdd(A=0, B=@n45.Result).Result
    ).Result
    Weight=@n41.OutputParam
   )
   > &RigUnit_ModifyTransforms(
    ItemToModify=((Item=(Type=Bone,Name="dyn_jacket_2_base"),Transform=(Rotation=(X=0,Y=0,Z=0,W=1),Translation=(X=-3,Y=0,Z=5),Scale3D=(X=1,Y=1,Z=1))))
    WeightMinimum=0
    ItemToModify.0.Transform.Translation.X=&RigVMFunction_MathDoubleAdd(A=0, B=&RigVMFunction_MathDoubleMul(A=-1.5, B=@n42.Value).Result).Result
    ItemToModify.0.Transform.Translation.Y=&RigVMFunction_MathDoubleAdd(A=0, B=&RigVMFunction_MathDoubleMul(A=0, B=@n42.Value).Result).Result
    ItemToModify.0.Transform.Translation.Z=&RigVMFunction_MathDoubleAdd(A=0, B=&RigVMFunction_MathDoubleMul(A=0, B=@n42.Value).Result).Result
    Weight=@n41.OutputParam
   )
   n46 = &RigUnit_SphericalPoseReader(
    ActiveRegionScaleFactors=(PositiveWidth=0.25,NegativeWidth=0.25,PositiveHeight=0.25,NegativeHeight=0.25)
    ActiveRegionSize=0.1
    Debug=(bDrawDebug=false,bDraw2D=False,bDrawLocalAxes=False,DebugScale=25,DebugSegments=20,DebugThickness=0.25)
    DriverAxis=(X=0,Y=-1,Z=0)
    DriverItem=(Type=Bone,Name="spine_04")
    FalloffRegionScaleFactors=(PositiveWidth=1,NegativeWidth=1,PositiveHeight=1,NegativeHeight=1)
    FalloffSize=0.15
    FlipHeightScaling=False
    FlipWidthScaling=False
    OptionalParentItem=(Type=Bone,Name="None")
    RotationOffset=(X=-110,Y=0,Z=0)
    ExecutePin=@RigVMFunction_Sequence_4.C
   )
   n47 = &RigVMFunction_MathDoubleMul(A=30, B=@n42.Value)
   n48 = &RigVMFunction_MathDoubleMul(A=20, B=@n42.Value)
   n49 = &RigVMFunction_MathDoubleMul(A=-20, B=@n42.Value)
   n50 = &RigVMFunction_AlphaInterp(InterpSpeedDecreasing=50, InterpSpeedIncreasing=300, Scale=1, bClampResult=false, Value=@n46.OutputParam)
   > &RigUnit_ModifyTransforms(
    ItemToModify=((Item=(Type=Bone,Name="dyn_jacket_1_base"),Transform=(Rotation=(X=0,Y=0,Z=0,W=1),Translation=(X=-3,Y=0,Z=5),Scale3D=(X=1,Y=1,Z=1))))
    WeightMinimum=0
    ItemToModify.0.Transform.Translation.X=&RigVMFunction_MathDoubleAdd(A=0, B=&RigVMFunction_MathDoubleMul(A=-3.5, B=@n42.Value).Result).Result
    ItemToModify.0.Transform.Translation.Y=&RigVMFunction_MathDoubleAdd(A=0, B=&RigVMFunction_MathDoubleMul(A=0, B=@n42.Value).Result).Result
    ItemToModify.0.Transform.Translation.Z=&RigVMFunction_MathDoubleAdd(A=0, B=&RigVMFunction_MathDoubleMul(A=0, B=@n42.Value).Result).Result
    ItemToModify.0.Transform.Rotation=&RigVMFunction_MathQuaternionFromEuler(
     Euler=(X=0,Y=0,Z=0)
     RotationOrder=XYZ
     Euler.X=&RigVMFunction_MathDoubleAdd(A=0, B=@n47.Result).Result
     Euler.Y=&RigVMFunction_MathDoubleAdd(A=0, B=@n48.Result).Result
     Euler.Z=&RigVMFunction_MathDoubleAdd(A=0, B=@n49.Result).Result
    ).Result
    Weight=@n50.Result
   )
   > &RigUnit_ModifyTransforms(
    ItemToModify=((Item=(Type=Bone,Name="dyn_jacket_2_base"),Transform=(Rotation=(X=0,Y=0,Z=0,W=1),Translation=(X=-3,Y=0,Z=5),Scale3D=(X=1,Y=1,Z=1))))
    WeightMinimum=0
    ItemToModify.0.Transform.Translation.Y=&RigVMFunction_MathDoubleAdd(A=0, B=&RigVMFunction_MathDoubleMul(A=0, B=@n42.Value).Result).Result
    ItemToModify.0.Transform.Translation.Z=&RigVMFunction_MathDoubleAdd(A=0, B=&RigVMFunction_MathDoubleMul(A=0, B=@n42.Value).Result).Result
    ItemToModify.0.Transform.Translation.X=&RigVMFunction_MathDoubleAdd(A=0, B=&RigVMFunction_MathDoubleMul(A=-2.5, B=@n42.Value).Result).Result
    Weight=@n50.Result
   )
   n51 = &RigUnit_SphericalPoseReader(
    ActiveRegionScaleFactors=(PositiveWidth=1,NegativeWidth=1,PositiveHeight=1,NegativeHeight=1)
    ActiveRegionSize=0.1
    Debug=(bDrawDebug=true,bDraw2D=False,bDrawLocalAxes=False,DebugScale=25,DebugSegments=20,DebugThickness=0.25)
    DriverAxis=(X=-1,Y=0,Z=0)
    DriverItem=(Type=Bone,Name="thigh_l")
    FalloffRegionScaleFactors=(PositiveWidth=1,NegativeWidth=1,PositiveHeight=1,NegativeHeight=1)
    FalloffSize=0.6
    FlipHeightScaling=False
    FlipWidthScaling=False
    OptionalParentItem=(Type=Bone,Name="None")
    RotationOffset=(X=120,Y=0,Z=-30)
    ExecutePin=@RigVMFunction_Sequence_3.B
   )
   n52 = RigVMVariableNode() {variable=is_crouch}
   n53 = &RigVMFunction_AlphaInterp(InterpSpeedDecreasing=50, InterpSpeedIncreasing=750, Scale=1, bClampResult=false, bInterpResult=false, Value=@n52.Value)
   n54 = &RigVMFunction_MathDoubleMul(A=30, B=@n53.Result)
   n55 = &RigVMFunction_MathDoubleMul(A=0, B=@n53.Result)
   n56 = &RigVMFunction_AlphaInterp(InterpSpeedDecreasing=50, InterpSpeedIncreasing=500, Scale=1, bClampResult=false, Value=@n51.OutputParam)
   > &RigUnit_ModifyTransforms(
    ItemToModify=((Item=(Type=Bone,Name="dyn_jacket_1_base"),Transform=(Rotation=(X=0,Y=0,Z=0,W=1),Translation=(X=-3,Y=0,Z=5),Scale3D=(X=1,Y=1,Z=1))))
    WeightMinimum=0
    ItemToModify.0.Transform.Translation.X=&RigVMFunction_MathDoubleAdd(A=-3, B=&RigVMFunction_MathDoubleMul(A=-6, B=@n52.Value).Result).Result
    ItemToModify.0.Transform.Translation.Y=&RigVMFunction_MathDoubleAdd(A=0, B=&RigVMFunction_MathDoubleMul(A=0, B=@n52.Value).Result).Result
    ItemToModify.0.Transform.Translation.Z=&RigVMFunction_MathDoubleAdd(A=2, B=&RigVMFunction_MathDoubleMul(A=2, B=@n52.Value).Result).Result
    ItemToModify.0.Transform.Rotation=&RigVMFunction_MathQuaternionFromEuler(
     Euler=(X=0,Y=0,Z=0)
     RotationOrder=XYZ
     Euler.X=&RigVMFunction_MathDoubleAdd(A=90, B=@n54.Result).Result
     Euler.Y=&RigVMFunction_MathDoubleAdd(A=-45, B=@n55.Result).Result
     Euler.Z=&RigVMFunction_MathDoubleAdd(A=-45, B=0).Result
    ).Result
    Weight=@n56.Result
   )
   n57 = RigVMVariableNode() {variable=is_crouch}
   n58 = &RigVMFunction_MathDoubleMul(A=45, B=@n57.Value)
   n59 = &RigVMFunction_MathDoubleMul(A=-10, B=@n57.Value)
   n60 = &RigVMFunction_MathDoubleMul(A=-20, B=@n57.Value)
   > &RigUnit_ModifyTransforms(
    ItemToModify=((Item=(Type=Bone,Name="dyn_jacket_2_base"),Transform=(Rotation=(X=0.0872,Y=0,Z=0,W=0.9962),Translation=(X=-3,Y=0,Z=0),Scale3D=(X=1,Y=1,Z=1))))
    WeightMinimum=0
    ItemToModify.0.Transform.Translation.X=&RigVMFunction_MathDoubleAdd(A=-1, B=&RigVMFunction_MathDoubleMul(A=-3, B=@n57.Value).Result).Result
    ItemToModify.0.Transform.Translation.Y=&RigVMFunction_MathDoubleAdd(A=2, B=&RigVMFunction_MathDoubleMul(A=1, B=@n57.Value).Result).Result
    ItemToModify.0.Transform.Translation.Z=&RigVMFunction_MathDoubleAdd(A=-5, B=&RigVMFunction_MathDoubleMul(A=3, B=@n57.Value).Result).Result
    ItemToModify.0.Transform.Rotation=&RigVMFunction_MathQuaternionFromEuler(
     Euler=(X=0,Y=0,Z=0)
     RotationOrder=XYZ
     Euler.X=&RigVMFunction_MathDoubleAdd(A=0, B=@n58.Result).Result
     Euler.Y=&RigVMFunction_MathDoubleAdd(A=-20, B=@n59.Result).Result
     Euler.Z=&RigVMFunction_MathDoubleAdd(A=-10, B=@n60.Result).Result
    ).Result
    Weight=@n56.Result
   )
   n61 = &RigVMFunction_AlphaInterp(InterpSpeedDecreasing=10, InterpSpeedIncreasing=10, Scale=1, bClampResult=False, Value=@n57.Value)
   > &RigUnit_ModifyTransforms(
    ItemToModify=((Item=(Type=Bone,Name="dyn_jacket_3_base"),Transform=(Rotation=(X=0,Y=0,Z=0,W=1),Translation=(X=0,Y=-1,Z=0),Scale3D=(X=1,Y=1,Z=1))))
    WeightMinimum=0
    Weight=@n61.Result
   )
   n62 = &RigUnit_SphericalPoseReader(
    ActiveRegionScaleFactors=(PositiveWidth=1,NegativeWidth=1,PositiveHeight=1,NegativeHeight=1)
    ActiveRegionSize=0.1
    Debug=(bDrawDebug=true,bDraw2D=False,bDrawLocalAxes=False,DebugScale=25,DebugSegments=20,DebugThickness=0.25)
    DriverAxis=(X=1,Y=0,Z=0)
    DriverItem=(Type=Bone,Name="thigh_r")
    FalloffRegionScaleFactors=(PositiveWidth=1,NegativeWidth=1,PositiveHeight=1,NegativeHeight=1)
    FalloffSize=0.5
    FlipHeightScaling=False
    FlipWidthScaling=False
    OptionalParentItem=(Type=Bone,Name="None")
    RotationOffset=(X=-75,Y=0,Z=-25)
    ExecutePin=@RigVMFunction_Sequence_3.C
   )
   n63 = RigVMVariableNode() {variable=is_crouch}
   n64 = &RigVMFunction_MathDoubleMul(A=-60, B=@n63.Value)
   n65 = &RigVMFunction_MathDoubleMul(A=30, B=@n63.Value)
   n66 = &RigVMFunction_MathDoubleMul(A=20, B=@n63.Value)
   n67 = &RigVMFunction_AlphaInterp(InterpSpeedDecreasing=1, InterpSpeedIncreasing=5, Scale=1, bClampResult=False, Value=@n62.OutputParam)
   > &RigUnit_ModifyTransforms(
    ItemToModify=((Item=(Type=Bone,Name="dyn_jacket_5_base"),Transform=(Rotation=(X=-0.6744,Y=0.2126,Z=0.0308,W=0.7064),Translation=(X=-5,Y=0,Z=-5),Scale3D=(X=1,Y=1,Z=1))))
    WeightMinimum=0
    ItemToModify.0.Transform.Translation.X=&RigVMFunction_MathDoubleAdd(A=-2, B=&RigVMFunction_MathDoubleMul(A=0, B=@n63.Value).Result).Result
    ItemToModify.0.Transform.Translation.Y=&RigVMFunction_MathDoubleAdd(A=0, B=&RigVMFunction_MathDoubleMul(A=0, B=@n63.Value).Result).Result
    ItemToModify.0.Transform.Translation.Z=&RigVMFunction_MathDoubleAdd(A=-3, B=&RigVMFunction_MathDoubleMul(A=0, B=@n63.Value).Result).Result
    ItemToModify.0.Transform.Rotation=&RigVMFunction_MathQuaternionFromEuler(
     Euler=(X=0,Y=0,Z=0)
     Euler.X=&RigVMFunction_MathDoubleAdd(A=-30, B=@n64.Result).Result
     Euler.Y=&RigVMFunction_MathDoubleAdd(A=30, B=@n65.Result).Result
     Euler.Z=&RigVMFunction_MathDoubleAdd(A=30, B=@n66.Result).Result
    ).Result
    Weight=@n67.Result
   )
   n68 = RigVMVariableNode() {variable=is_crouch}
   n69 = &RigVMFunction_MathDoubleMul(A=0, B=@n68.Value)
   n70 = &RigVMFunction_MathDoubleMul(A=-10, B=@n68.Value)
   n71 = &RigVMFunction_MathDoubleMul(A=-10, B=@n68.Value)
   > &RigUnit_ModifyTransforms(
    ItemToModify=((Item=(Type=Bone,Name="dyn_jacket_4_base"),Transform=(Rotation=(X=-0.13,Y=-0.0864,Z=-0.0114,W=0.9877),Translation=(X=-2,Y=0,Z=0),Scale3D=(X=1,Y=1,Z=1))))
    WeightMinimum=0
    ItemToModify.0.Transform.Translation.X=&RigVMFunction_MathDoubleAdd(A=-2, B=&RigVMFunction_MathDoubleMul(A=-1, B=@n68.Value).Result).Result
    ItemToModify.0.Transform.Translation.Y=&RigVMFunction_MathDoubleAdd(A=0, B=&RigVMFunction_MathDoubleMul(A=0, B=@n68.Value).Result).Result
    ItemToModify.0.Transform.Translation.Z=&RigVMFunction_MathDoubleAdd(A=0, B=&RigVMFunction_MathDoubleMul(A=0, B=@n68.Value).Result).Result
    ItemToModify.0.Transform.Rotation=&RigVMFunction_MathQuaternionFromEuler(
     Euler=(X=0,Y=0,Z=0)
     Euler.X=&RigVMFunction_MathDoubleAdd(A=30, B=@n69.Result).Result
     Euler.Y=&RigVMFunction_MathDoubleAdd(A=-10, B=@n70.Result).Result
     Euler.Z=&RigVMFunction_MathDoubleAdd(A=-20, B=@n71.Result).Result
    ).Result
    Weight=@n67.Result
   )
   n72 = &RigVMFunction_Sequence(ExecuteContext=@RigVMFunction_Sequence_3.D)
   n73 = &RigVMFunction_AlphaInterp(InterpSpeedDecreasing=10, InterpSpeedIncreasing=10, Scale=1, bClampResult=False, Value=RigVMVariableNode() {variable=is_emote}.Value)
   &RigUnit_ModifyTransforms(
    ItemToModify=((Item=(Type=Bone,Name="dyn_jacket_1_base"),Transform=(Rotation=(X=0,Y=0,Z=0,W=1),Translation=(X=-1,Y=3,Z=4),Scale3D=(X=1,Y=1,Z=1))))
    WeightMinimum=0
    Weight=@n73.Result
    ItemToModify.0.Transform.Rotation=&RigVMFunction_MathQuaternionFromEuler(Euler=(X=-10,Y=-20,Z=-30)).Result
    ExecutePin=@n72.A
   )
   > &RigUnit_ModifyTransforms(
    ItemToModify=((Item=(Type=Bone,Name="dyn_jacket_2_base"),Transform=(Rotation=(X=-0.13,Y=-0.0864,Z=-0.0114,W=0.9877),Translation=(X=-1,Y=0,Z=1),Scale3D=(X=1,Y=1,Z=1))))
    WeightMinimum=0
    ItemToModify.0.Transform.Rotation=&RigVMFunction_MathQuaternionFromEuler(Euler=(X=0,Y=0,Z=0)).Result
    Weight=@n73.Result
   )
   > &RigUnit_ModifyTransforms(
    ItemToModify=((Item=(Type=Bone,Name="dyn_jacket_5_base"),Transform=(Rotation=(X=-0.6744,Y=0.2126,Z=0.0308,W=0.7064),Translation=(X=0,Y=2,Z=-2),Scale3D=(X=1,Y=1,Z=1))))
    WeightMinimum=0
    ItemToModify.0.Transform.Rotation=&RigVMFunction_MathQuaternionFromEuler(Euler=(X=0,Y=0,Z=0)).Result
    Weight=@n73.Result
   )
   > &RigUnit_ModifyTransforms(
    ItemToModify=((Item=(Type=Bone,Name="dyn_jacket_4_base"),Transform=(Rotation=(X=-0.13,Y=-0.0864,Z=-0.0114,W=0.9877),Translation=(X=0,Y=0,Z=-1),Scale3D=(X=1,Y=1,Z=1))))
    WeightMinimum=0
    ItemToModify.0.Transform.Rotation=&RigVMFunction_MathQuaternionFromEuler(Euler=(X=0,Y=0,Z=0)).Result
    Weight=@n73.Result
   )
   n74 = &RigUnit_SphericalPoseReader(
    ActiveRegionScaleFactors=(PositiveWidth=1,NegativeWidth=1,PositiveHeight=2,NegativeHeight=2)
    ActiveRegionSize=0.1
    Debug=(bDrawDebug=true,bDraw2D=False,bDrawLocalAxes=False,DebugScale=25,DebugSegments=20,DebugThickness=0.25)
    DriverAxis=(X=0,Y=-1,Z=0)
    DriverItem=(Type=Bone,Name="pelvis")
    FalloffRegionScaleFactors=(PositiveWidth=1,NegativeWidth=1,PositiveHeight=1,NegativeHeight=1)
    FalloffSize=0.45
    FlipHeightScaling=False
    FlipWidthScaling=False
    OptionalParentItem=(Type=Bone,Name="None")
    RotationOffset=(X=-90,Y=0,Z=60)
    ExecutePin=@n72.B
   )
   n75 = &RigVMFunction_AlphaInterp(InterpSpeedDecreasing=10, InterpSpeedIncreasing=10, Scale=1, bClampResult=False, Value=RigVMVariableNode() {variable=is_sky_diving}.Value)
   n76 = &RigVMFunction_AlphaInterp(InterpSpeedDecreasing=50, InterpSpeedIncreasing=300, Scale=1, bClampResult=false, Value=@n74.OutputParam)
   n77 = &RigVMFunction_MathDoubleMul(B=@n75.Result, A=@n76.Result)
   > &RigUnit_ModifyTransforms(
    ItemToModify=((Item=(Type=Bone,Name="dyn_jacket_1_base"),Transform=(Rotation=(X=0,Y=0,Z=0,W=1),Translation=(X=-1.5,Y=2,Z=4),Scale3D=(X=1,Y=1,Z=1))))
    WeightMinimum=0
    Weight=@n77.Result
   )
   > &RigUnit_ModifyTransforms(
    ItemToModify=((Item=(Type=Bone,Name="dyn_jacket_2_base"),Transform=(Rotation=(X=-0.0334,Y=-0.0805,Z=-0.3812,W=0.9204),Translation=(X=0,Y=5,Z=-5),Scale3D=(X=1,Y=1,Z=1))))
    WeightMinimum=0
    Weight=@n77.Result
   )
   > &RigUnit_ModifyTransforms(
    ItemToModify=((Item=(Type=Bone,Name="dyn_jacket_3_base"),Transform=(Rotation=(X=0.2296,Y=-0.1195,Z=-0.446,W=0.8568),Translation=(X=0,Y=-2,Z=0),Scale3D=(X=1,Y=1,Z=1))))
    WeightMinimum=0
    Weight=@n77.Result
   )
   > &RigUnit_ModifyTransforms(
    ItemToModify=((Item=(Type=Bone,Name="dyn_jacket_4_base"),Transform=(Rotation=(X=0,Y=0,Z=-0.3827,W=0.9239),Translation=(X=0,Y=-3,Z=-2),Scale3D=(X=1,Y=1,Z=1))))
    WeightMinimum=0
    Weight=@n77.Result
   )
   RigVMFunction_Sequence_7 = RigVMAggregateNode(ExecuteContext=())
   n78 = &RigVMFunction_AlphaInterp(InterpSpeedDecreasing=10, InterpSpeedIncreasing=10, Scale=1, bClampResult=False, Value=RigVMVariableNode() {variable=is_emote}.Value)
   &RigUnit_ModifyTransforms(
    ItemToModify=((Item=(Type=Bone,Name="dyn_jacket_1_base"),Transform=(Rotation=(X=0,Y=0,Z=0,W=1),Translation=(X=-1,Y=3,Z=4),Scale3D=(X=1,Y=1,Z=1))))
    WeightMinimum=0
    Weight=@n78.Result
    ItemToModify.0.Transform.Rotation=&RigVMFunction_MathQuaternionFromEuler(Euler=(X=0,Y=0,Z=0)).Result
    ExecutePin=@RigVMFunction_Sequence_7.A
   )
   > &RigUnit_ModifyTransforms(
    ItemToModify=((Item=(Type=Bone,Name="dyn_jacket_2_base"),Transform=(Rotation=(X=-0.13,Y=-0.0864,Z=-0.0114,W=0.9877),Translation=(X=-1,Y=0,Z=1),Scale3D=(X=1,Y=1,Z=1))))
    WeightMinimum=0
    ItemToModify.0.Transform.Rotation=&RigVMFunction_MathQuaternionFromEuler(Euler=(X=0,Y=0,Z=0)).Result
    Weight=@n78.Result
   )
   > &RigUnit_ModifyTransforms(
    ItemToModify=((Item=(Type=Bone,Name="dyn_jacket_5_base"),Transform=(Rotation=(X=-0.6744,Y=0.2126,Z=0.0308,W=0.7064),Translation=(X=0,Y=2,Z=-2),Scale3D=(X=1,Y=1,Z=1))))
    WeightMinimum=0
    ItemToModify.0.Transform.Rotation=&RigVMFunction_MathQuaternionFromEuler(Euler=(X=0,Y=0,Z=0)).Result
    Weight=@n78.Result
   )
   > &RigUnit_ModifyTransforms(
    ItemToModify=((Item=(Type=Bone,Name="dyn_jacket_4_base"),Transform=(Rotation=(X=-0.13,Y=-0.0864,Z=-0.0114,W=0.9877),Translation=(X=0,Y=0,Z=-1),Scale3D=(X=1,Y=1,Z=1))))
    WeightMinimum=0
    ItemToModify.0.Transform.Rotation=&RigVMFunction_MathQuaternionFromEuler(Euler=(X=0,Y=0,Z=0)).Result
    Weight=@n78.Result
   )
   n79 = &RigUnit_SphericalPoseReader(
    ActiveRegionScaleFactors=(PositiveWidth=1,NegativeWidth=1,PositiveHeight=2,NegativeHeight=2)
    ActiveRegionSize=0.1
    Debug=(bDrawDebug=false,bDraw2D=False,bDrawLocalAxes=False,DebugScale=25,DebugSegments=20,DebugThickness=0.25)
    DriverAxis=(X=0,Y=-1,Z=0)
    DriverItem=(Type=Bone,Name="pelvis")
    FalloffRegionScaleFactors=(PositiveWidth=1,NegativeWidth=1,PositiveHeight=1,NegativeHeight=1)
    FalloffSize=0.45
    FlipHeightScaling=False
    FlipWidthScaling=False
    OptionalParentItem=(Type=Bone,Name="None")
    RotationOffset=(X=-90,Y=0,Z=60)
    ExecutePin=@RigVMFunction_Sequence_7.B
   )
   n80 = &RigVMFunction_AlphaInterp(InterpSpeedDecreasing=10, InterpSpeedIncreasing=10, Scale=1, bClampResult=False, Value=RigVMVariableNode() {variable=is_sky_diving}.Value)
   n81 = &RigVMFunction_AlphaInterp(InterpSpeedDecreasing=50, InterpSpeedIncreasing=300, Scale=1, bClampResult=false, Value=@n79.OutputParam)
   n82 = &RigVMFunction_MathDoubleMul(B=@n80.Result, A=@n81.Result)
   > &RigUnit_ModifyTransforms(
    ItemToModify=((Item=(Type=Bone,Name="dyn_jacket_1_base"),Transform=(Rotation=(X=0,Y=0,Z=-0.1305,W=0.9914),Translation=(X=-1.5,Y=3,Z=9),Scale3D=(X=1,Y=1,Z=1))))
    WeightMinimum=0
    Weight=@n82.Result
   )
   > &RigUnit_ModifyTransforms(
    ItemToModify=((Item=(Type=Bone,Name="dyn_jacket_2_base"),Transform=(Rotation=(X=-0.0665,Y=-0.1604,Z=-0.3769,W=0.9098),Translation=(X=0,Y=5,Z=-5),Scale3D=(X=1,Y=1,Z=1))))
    WeightMinimum=0
    Weight=@n82.Result
   )
   > &RigUnit_ModifyTransforms(
    ItemToModify=((Item=(Type=Bone,Name="dyn_jacket_3_base"),Transform=(Rotation=(X=0.113,Y=-0.0653,Z=-0.4957,W=0.8586),Translation=(X=0,Y=-2,Z=0),Scale3D=(X=1,Y=1,Z=1))))
    WeightMinimum=0
    Weight=@n82.Result
   )
   > &RigUnit_ModifyTransforms(
    ItemToModify=((Item=(Type=Bone,Name="dyn_jacket_4_base"),Transform=(Rotation=(X=0.1935,Y=0.0934,Z=-0.4709,W=0.8556),Translation=(X=0,Y=-3,Z=-2),Scale3D=(X=1,Y=1,Z=1))))
    WeightMinimum=0
    Weight=@n82.Result
   )
   > &RigUnit_ModifyTransforms(
    ItemToModify=((Item=(Type=Bone,Name="dyn_jacket_5_base"),Transform=(Rotation=(X=0,Y=0,Z=-0.1305,W=0.9914),Translation=(X=0,Y=0,Z=0),Scale3D=(X=1,Y=1,Z=1))))
    WeightMinimum=0
    Weight=@n82.Result
   )
   > n83 = &RigUnit_SphericalPoseReader(
    ActiveRegionScaleFactors=(PositiveWidth=1,NegativeWidth=1,PositiveHeight=2,NegativeHeight=2)
    ActiveRegionSize=0.1
    Debug=(bDrawDebug=false,bDraw2D=False,bDrawLocalAxes=False,DebugScale=25,DebugSegments=20,DebugThickness=0.25)
    DriverAxis=(X=0,Y=-1,Z=0)
    DriverItem=(Type=Bone,Name="pelvis")
    FalloffRegionScaleFactors=(PositiveWidth=1,NegativeWidth=1,PositiveHeight=1,NegativeHeight=1)
    FalloffSize=0.35
    FlipHeightScaling=False
    FlipWidthScaling=False
    OptionalParentItem=(Type=Bone,Name="None")
    RotationOffset=(X=-90,Y=0,Z=150)
   )
   n84 = &RigVMFunction_MathDoubleMul(B=@n80.Result, A=@n83.OutputParam)
   > &RigUnit_ModifyTransforms(
    ItemToModify=((Item=(Type=Bone,Name="dyn_jacket_1_base"),Transform=(Rotation=(X=0.0227,Y=0.1722,Z=-0.1285,W=0.9764),Translation=(X=0,Y=0,Z=0),Scale3D=(X=1,Y=1,Z=1))))
    WeightMinimum=0
    Weight=@n84.Result
   )
   > &RigUnit_ModifyTransforms(
    ItemToModify=((Item=(Type=Bone,Name="dyn_jacket_2_base"),Transform=(Rotation=(X=-0.0131,Y=-0.0416,Z=-0.3004,W=0.9528),Translation=(X=0,Y=2.5,Z=-2.5),Scale3D=(X=1,Y=1,Z=1))))
    WeightMinimum=0
    Weight=@n84.Result
   )
   > &RigUnit_ModifyTransforms(
    ItemToModify=((Item=(Type=Bone,Name="dyn_jacket_3_base"),Transform=(Rotation=(X=0.0114,Y=0.0864,Z=-0.13,W=0.9877),Translation=(X=0,Y=-2,Z=0),Scale3D=(X=1,Y=1,Z=1))))
    WeightMinimum=0
    Weight=@n84.Result
   )
   > &RigUnit_ModifyTransforms(
    ItemToModify=((Item=(Type=Bone,Name="dyn_jacket_4_base"),Transform=(Rotation=(X=0.2468,Y=-0.0778,Z=-0.2905,W=0.9212),Translation=(X=0,Y=-1.5,Z=-1),Scale3D=(X=1,Y=1,Z=1))))
    WeightMinimum=0
    Weight=@n84.Result
   )
   > &RigUnit_ModifyTransforms(
    ItemToModify=((Item=(Type=Bone,Name="dyn_jacket_5_base"),Transform=(Rotation=(X=-0.0376,Y=-0.1695,Z=-0.2132,W=0.9615),Translation=(X=0,Y=0,Z=0),Scale3D=(X=1,Y=1,Z=1))))
    WeightMinimum=0
    Weight=@n84.Result
   )
   n85 = &RigUnit_SphericalPoseReader(
    ActiveRegionScaleFactors=(PositiveWidth=0.25,NegativeWidth=0.25,PositiveHeight=0.25,NegativeHeight=0.25)
    ActiveRegionSize=0.1
    Debug=(bDrawDebug=false,bDraw2D=False,bDrawLocalAxes=False,DebugScale=25,DebugSegments=20,DebugThickness=0.25)
    DriverAxis=(X=0,Y=-1,Z=0)
    DriverItem=(Type=Bone,Name="spine_04")
    FalloffRegionScaleFactors=(PositiveWidth=1,NegativeWidth=1,PositiveHeight=1,NegativeHeight=1)
    FalloffSize=0.1
    FlipHeightScaling=False
    FlipWidthScaling=False
    OptionalParentItem=(Type=Bone,Name="None")
    RotationOffset=(X=-115,Y=0,Z=-5)
    ExecutePin=@RigVMFunction_Sequence_7.C
   )
   n86 = &RigVMFunction_AlphaInterp(InterpSpeedDecreasing=10, InterpSpeedIncreasing=10, Scale=1, bClampResult=False, Value=RigVMVariableNode() {variable=is_crouch}.Value)
   > &RigUnit_ModifyTransforms(
    ItemToModify=((Item=(Type=Bone,Name="dyn_jacket_1_base"),Transform=(Rotation=(X=0,Y=0,Z=0,W=1),Translation=(X=0,Y=0,Z=3),Scale3D=(X=1,Y=1,Z=1))))
    WeightMinimum=0
    ItemToModify.0.Transform.Rotation=&RigVMFunction_MathQuaternionFromEuler(Euler=(X=0,Y=-45,Z=30), RotationOrder=XYZ).Result
    Weight=&RigVMFunction_MathDoubleMul(B=@n86.Result, A=@n85.OutputParam).Result
   )
   n87 = &RigUnit_SphericalPoseReader(
    ActiveRegionScaleFactors=(PositiveWidth=0.25,NegativeWidth=0.25,PositiveHeight=0.25,NegativeHeight=0.25)
    ActiveRegionSize=0.1
    Debug=(bDrawDebug=true,bDraw2D=False,bDrawLocalAxes=False,DebugScale=25,DebugSegments=20,DebugThickness=0.25)
    DriverAxis=(X=0,Y=-1,Z=0)
    DriverItem=(Type=Bone,Name="pelvis")
    ExecutePin=()
    FalloffRegionScaleFactors=(PositiveWidth=1,NegativeWidth=1,PositiveHeight=0.1,NegativeHeight=1)
    FalloffSize=0.1
    FlipHeightScaling=False
    FlipWidthScaling=False
    OptionalParentItem=(Type=Bone,Name="None")
    RotationOffset=(X=-90,Y=0,Z=0)
   )
   n88 = &RigVMFunction_MathDoubleMul(A=-30, B=@n42.Value)
   n89 = &RigVMFunction_MathDoubleMul(A=45, B=@n42.Value)
   n90 = &RigVMFunction_MathDoubleMul(A=30, B=@n42.Value)
   > &RigUnit_ModifyTransforms(
    ItemToModify=((Item=(Type=Bone,Name="dyn_jacket_1_base"),Transform=(Rotation=(X=0,Y=0,Z=0,W=1),Translation=(X=-3,Y=0,Z=5),Scale3D=(X=1,Y=1,Z=1))))
    WeightMinimum=0
    ItemToModify.0.Transform.Translation.X=&RigVMFunction_MathDoubleAdd(A=0, B=&RigVMFunction_MathDoubleMul(A=-5, B=@n42.Value).Result).Result
    ItemToModify.0.Transform.Translation.Y=&RigVMFunction_MathDoubleAdd(A=0, B=&RigVMFunction_MathDoubleMul(A=0, B=@n42.Value).Result).Result
    ItemToModify.0.Transform.Translation.Z=&RigVMFunction_MathDoubleAdd(A=0, B=&RigVMFunction_MathDoubleMul(A=0, B=@n42.Value).Result).Result
    ItemToModify.0.Transform.Rotation=&RigVMFunction_MathQuaternionFromEuler(
     Euler=(X=0,Y=0,Z=0)
     RotationOrder=XYZ
     Euler.X=&RigVMFunction_MathDoubleAdd(A=0, B=@n88.Result).Result
     Euler.Y=&RigVMFunction_MathDoubleAdd(A=0, B=@n89.Result).Result
     Euler.Z=&RigVMFunction_MathDoubleAdd(A=0, B=@n90.Result).Result
    ).Result
    Weight=@n87.OutputParam
   )
   > &RigUnit_ModifyTransforms(
    ItemToModify=((Item=(Type=Bone,Name="dyn_jacket_2_base"),Transform=(Rotation=(X=0,Y=0,Z=0,W=1),Translation=(X=-3,Y=0,Z=5),Scale3D=(X=1,Y=1,Z=1))))
    WeightMinimum=0
    ItemToModify.0.Transform.Translation.X=&RigVMFunction_MathDoubleAdd(A=0, B=&RigVMFunction_MathDoubleMul(A=-3.5, B=@n42.Value).Result).Result
    ItemToModify.0.Transform.Translation.Y=&RigVMFunction_MathDoubleAdd(A=0, B=&RigVMFunction_MathDoubleMul(A=0, B=@n42.Value).Result).Result
    ItemToModify.0.Transform.Translation.Z=&RigVMFunction_MathDoubleAdd(A=0, B=&RigVMFunction_MathDoubleMul(A=0, B=@n42.Value).Result).Result
    Weight=@n87.OutputParam
   )
   &RigVMFunction_MathDoubleAdd(A=-30, B=&RigVMFunction_MathDoubleMul(A=-60, B=1).Result)
   &RigVMFunction_MathDoubleAdd(A=30, B=&RigVMFunction_MathDoubleMul(A=30, B=1).Result)
   &RigVMFunction_MathDoubleAdd(A=30, B=&RigVMFunction_MathDoubleMul(A=20, B=1).Result)
   &RigVMFunction_MathDoubleAdd(A=30, B=&RigVMFunction_MathDoubleMul(A=0, B=1).Result)
   &RigVMFunction_MathDoubleAdd(A=-10, B=&RigVMFunction_MathDoubleMul(A=-10, B=1).Result)
   &RigVMFunction_MathDoubleAdd(A=-20, B=&RigVMFunction_MathDoubleMul(A=-10, B=1).Result)
   &RigUnit_ModifyTransforms(
    ExecutePin=()
    ItemToModify=((Item=(Type=Bone,Name="dyn_jacket_1_base"),Transform=(Rotation=(X=0,Y=0,Z=0,W=1),Translation=(X=-2,Y=15,Z=4),Scale3D=(X=1,Y=1,Z=1))))
    Weight=1
    WeightMinimum=0
   )
   &RigVMFunction_MathDoubleMul(A=-10, B=@n53.Result)
  ]}
]
hierarchy [
 root {loc=<0 0 0>, rot=<0 0 0 1>}
 pelvis {parent=root, loc=<0 -0.7071 93.1191>, rot=<0.0127 0.707 -0.0127 -0.707>}
 spine_01 {parent=pelvis, loc=<6.7494 -0.6388 0>, rot=<0 0 0.034 -0.9994>}
 spine_02 {parent=spine_01, loc=<7.2518 -0.035 0>, rot=<0 0 0.0238 -0.9997>}
 spine_03 {parent=spine_02, loc=<6.0104 0 0>, rot=<0 0 0.0149 -0.9999>}
 spine_04 {parent=spine_03, loc=<8.1024 0 0>, rot=<0 0 -0.0132 -0.9999>}
 spine_05 {parent=spine_04, loc=<10.0146 0 0>, rot=<0 0 -0.0015 -1>}
 neck_01 {parent=spine_05, loc=<11.3138 0.1529 0>, rot=<0 0 -0.2003 -0.9797>}
 neck_02 {parent=neck_01, loc=<4.5627 0 0>, rot=<0 0 0.0504 -0.9987>}
 head {parent=neck_02, loc=<5.3914 0 0>, rot=<0 0 0.115 -0.9934>}
 attach_cape {parent=spine_05, loc=<9.017 -11.714 0>, rot=<0.0616 -0.7044 -0.0616 -0.7044>}
 cape {parent=attach_cape, type=Null, loc=<0 0.6 0>, rot=<0 0 0 1>}
 attach_backpack {parent=spine_05, loc=<-0.123 -13.473 0>, rot=<-0.0093 -0.707 0.0093 -0.707>}
 Backpack_BR {parent=attach_backpack, type=Null, loc=<0 1 0>, rot=<-0.0305 0 0 0.9995>}
 clavicle_l {parent=spine_05, loc=<4.3575 -0.1466 -2.3223>, rot=<0.0465 -0.7289 -0.009 -0.6829>}
 upperarm_l {parent=clavicle_l, loc=<15.5111 0 0>, rot=<0.0226 -0.3234 -0.0088 -0.9459>}
 lowerarm_l {parent=upperarm_l, loc=<26.6514 0 0>, rot=<0.0391 -0.0099 -0.2453 -0.9686>}
 hand_l {parent=lowerarm_l, loc=<22.3491 0 0>, rot=<-0.7081 -0.0507 0.0975 -0.6975>}
 thumb_01_l {parent=hand_l, loc=<1.3118 -0.3191 -2.459>, rot=<0.6546 -0.4042 0.0664 -0.6354>}
 thumb_02_l {parent=thumb_01_l, loc=<5.9336 -0.0448 0.0401>, rot=<-0.0037 0.0213 0.298 -0.9543>}
 thumb_03_l {parent=thumb_02_l, loc=<2.2223 0.3682 -0.0907>, rot=<0.0118 -0.013 0.0066 -0.9998>}
 index_metacarpal_l {parent=hand_l, loc=<1.6535 0.0183 -2.96>, rot=<0.0158 -0.0124 -0.0173 -0.9996>}
 index_01_l {parent=index_metacarpal_l, loc=<8.3008 0.0357 -0.0524>, rot=<-0.0044 0.0485 0.0924 -0.9945>}
 index_02_l {parent=index_01_l, loc=<3.3587 -0.3004 0.0452>, rot=<-0.002 0.0027 0.3225 -0.9466>}
 index_03_l {parent=index_02_l, loc=<2.5578 -0.123 -0.0252>, rot=<-0.002 0.0067 0.2558 -0.9667>}
 middle_metacarpal_l {parent=hand_l, loc=<2.1856 -0.1725 -0.5984>, rot=<0.006 0.0168 -0.0329 -0.9993>}
 middle_01_l {parent=middle_metacarpal_l, loc=<7.3275 -0.2447 0.0129>, rot=<-0.0208 0.0042 0.0858 -0.9961>}
 middle_02_l {parent=middle_01_l, loc=<4.0123 0.1401 0.1175>, rot=<-0.0218 0.0258 0.3512 -0.9357>}
 middle_03_l {parent=middle_02_l, loc=<2.6077 0.0534 0.1373>, rot=<0.0082 0.0188 0.1708 -0.9851>}
 ring_metacarpal_l {parent=hand_l, loc=<2.2074 -0.3625 1.3638>, rot=<-0.0284 0.0722 -0.0458 -0.9959>}
 ring_01_l {parent=ring_metacarpal_l, loc=<7.0295 -0.6197 -0.1446>, rot=<-0.0489 -0.0252 0.1892 -0.9804>}
 ring_02_l {parent=ring_01_l, loc=<3.6174 0.4647 -0.1027>, rot=<0 0 0.2806 -0.9598>}
 ring_03_l {parent=ring_02_l, loc=<2.4152 0.114 0.0199>, rot=<0 0 0.1189 -0.9929>}
 pinky_metacarpal_l {parent=hand_l, loc=<1.7616 -0.9753 3.3814>, rot=<-0.0397 0.0814 -0.069 -0.9935>}
 pinky_01_l {parent=pinky_metacarpal_l, loc=<7.1709 -0.4831 -0.1094>, rot=<-0.0654 -0.0035 0.2242 -0.9723>}
 pinky_02_l {parent=pinky_01_l, loc=<2.8274 0.0893 -0.1043>, rot=<0.0037 0.0011 0.2854 -0.9584>}
 pinky_03_l {parent=pinky_02_l, loc=<2.2046 0.0307 -0.0078>, rot=<-0.0002 0.0098 0.1445 -0.9895>}
 deform_wrist_dn_base_l {parent=hand_l, loc=<0 -2.2342 0>, rot=<0 0 0 -1>}
 deform_wrist_dn_l {parent=deform_wrist_dn_base_l, loc=<0 0 0>, rot=<0 0 0 1>}
 deform_wrist_up_base_l {parent=hand_l, loc=<0 2.2682 0>, rot=<0 0 0 -1>}
 deform_wrist_up_l {parent=deform_wrist_up_base_l, loc=<0 0 0>, rot=<0 0 0 1>}
 weapon_l {parent=hand_l, loc=<-0.1974 -3.4111 0.3811>, rot=<0 0 0 1>}
 hand_attach_l {parent=hand_l, loc=<8.5 -3 0>, rot=<0.7071 0.7071 0 0>}
 lowerarm_twist_02_l {parent=lowerarm_l, loc=<9.6199 0 0>, rot=<0 0 0 -1>}
 lowerarm_twist_01_l {parent=lowerarm_l, loc=<17.8549 0 0>, rot=<0 0 0 -1>}
 deform_elbow_in_base_l {parent=lowerarm_l, loc=<0.2112 5.259 0.8274>, rot=<0 0 0 -1>}
 deform_elbow_in_l {parent=deform_elbow_in_base_l, loc=<0 0 0>, rot=<0 0 0 1>}
 deform_elbow_out_base_l {parent=lowerarm_l, loc=<0 -4 0>, rot=<0 0 0 -1>}
 deform_elbow_out_l {parent=deform_elbow_out_base_l, loc=<0 0 0>, rot=<0 0 0 1>}
 upperarm_twist_01_l {parent=upperarm_l, loc=<0.0041 0 0>, rot=<0 0 0 -1>}
 deform_upperarm_bk_base_l {parent=upperarm_twist_01_l, loc=<1.0813 -7.2193 -1.9453>, rot=<0 0 0 1>}
 deform_upperarm_bk_l {parent=deform_upperarm_bk_base_l, loc=<0 0 0>, rot=<0 0 0 1>}
 deform_upperarm_fr_base_l {parent=upperarm_twist_01_l, loc=<-0.3691 7.0506 -2.765>, rot=<0 0 0 1>}
 deform_upperarm_fr_l {parent=deform_upperarm_fr_base_l, loc=<0 0 0>, rot=<0 0 0 1>}
 deform_upperarm_up_base_l {parent=upperarm_twist_01_l, loc=<-3.2216 -1.1391 3.3311>, rot=<0 0 0 1>}
 deform_upperarm_up_l {parent=deform_upperarm_up_base_l, loc=<0 0 0>, rot=<0 0 0 1>}
 upperarm_twist_02_l {parent=upperarm_l, loc=<13.7941 0 0>, rot=<0 0 0 -1>}
 clavicle_r {parent=spine_05, loc=<4.3575 -0.1466 2.3223>, rot=<0.7289 0.0465 -0.6829 0.009>}
 upperarm_r {parent=clavicle_r, loc=<-15.5111 0 0>, rot=<0.0226 -0.3234 -0.0088 -0.9459>}
 lowerarm_r {parent=upperarm_r, loc=<-26.6514 0 0>, rot=<0.0391 -0.0099 -0.2453 -0.9686>}
 hand_r {parent=lowerarm_r, loc=<-22.3491 0 0>, rot=<-0.7081 -0.0507 0.0975 -0.6975>}
 thumb_01_r {parent=hand_r, loc=<-1.3118 0.3191 2.459>, rot=<0.6546 -0.4042 0.0664 -0.6354>}
 thumb_02_r {parent=thumb_01_r, loc=<-5.9336 0.0448 -0.0401>, rot=<-0.0037 0.0213 0.298 -0.9543>}
 thumb_03_r {parent=thumb_02_r, loc=<-2.2223 -0.3682 0.0907>, rot=<0.0118 -0.013 0.0066 -0.9998>}
 index_metacarpal_r {parent=hand_r, loc=<-1.6535 -0.0183 2.96>, rot=<0.0158 -0.0124 -0.0173 -0.9996>}
 index_01_r {parent=index_metacarpal_r, loc=<-8.3008 -0.0357 0.0524>, rot=<-0.0044 0.0485 0.0924 -0.9945>}
 index_02_r {parent=index_01_r, loc=<-3.3587 0.3004 -0.0452>, rot=<-0.002 0.0027 0.3225 -0.9466>}
 index_03_r {parent=index_02_r, loc=<-2.5578 0.123 0.0252>, rot=<-0.002 0.0067 0.2558 -0.9667>}
 middle_metacarpal_r {parent=hand_r, loc=<-2.1856 0.1725 0.5984>, rot=<0.006 0.0168 -0.0329 -0.9993>}
 middle_01_r {parent=middle_metacarpal_r, loc=<-7.3275 0.2447 -0.0129>, rot=<-0.0208 0.0042 0.0858 -0.9961>}
 middle_02_r {parent=middle_01_r, loc=<-4.0123 -0.1401 -0.1175>, rot=<-0.0218 0.0258 0.3512 -0.9357>}
 middle_03_r {parent=middle_02_r, loc=<-2.6077 -0.0534 -0.1373>, rot=<0.0082 0.0188 0.1708 -0.9851>}
 ring_metacarpal_r {parent=hand_r, loc=<-2.2074 0.3625 -1.3638>, rot=<-0.0284 0.0722 -0.0458 -0.9959>}
 ring_01_r {parent=ring_metacarpal_r, loc=<-7.0295 0.6197 0.1446>, rot=<-0.0489 -0.0252 0.1892 -0.9804>}
 ring_02_r {parent=ring_01_r, loc=<-3.6174 -0.4647 0.1027>, rot=<0 0 0.2806 -0.9598>}
 ring_03_r {parent=ring_02_r, loc=<-2.4152 -0.114 -0.0199>, rot=<0 0 0.1189 -0.9929>}
 pinky_metacarpal_r {parent=hand_r, loc=<-1.7616 0.9753 -3.3814>, rot=<-0.0397 0.0814 -0.069 -0.9935>}
 pinky_01_r {parent=pinky_metacarpal_r, loc=<-7.1709 0.4831 0.1094>, rot=<-0.0654 -0.0035 0.2242 -0.9723>}
 pinky_02_r {parent=pinky_01_r, loc=<-2.8274 -0.0893 0.1043>, rot=<0.0037 0.0011 0.2854 -0.9584>}
 pinky_03_r {parent=pinky_02_r, loc=<-2.2046 -0.0307 0.0078>, rot=<-0.0002 0.0098 0.1445 -0.9895>}
 deform_wrist_dn_base_r {parent=hand_r, loc=<-0.0004 2.2345 -0.0001>, rot=<0 0 0 -1>}
 deform_wrist_dn_r {parent=deform_wrist_dn_base_r, loc=<0 0 0>, rot=<0 0 0 1>}
 deform_wrist_up_base_r {parent=hand_r, loc=<0.0003 -2.2684 0.0001>, rot=<0 0 0 -1>}
 deform_wrist_up_r {parent=deform_wrist_up_base_r, loc=<0 0 0>, rot=<0 0 0 1>}
 weapon_r {parent=hand_r, loc=<0.1974 3.4111 -0.3811>, rot=<0 0 0 1>}
 hand_attach_r {parent=hand_r, loc=<-8.5001 2.9999 0>, rot=<0 0 -0.7071 -0.7071>}
 lowerarm_twist_02_r {parent=lowerarm_r, loc=<-9.6199 0 0>, rot=<0 0 0 -1>}
 lowerarm_twist_01_r {parent=lowerarm_r, loc=<-17.8549 0 0>, rot=<0 0 0 -1>}
 deform_elbow_in_base_r {parent=lowerarm_r, loc=<-0.2111 -5.2591 -0.8275>, rot=<0 0 0 -1>}
 deform_elbow_in_r {parent=deform_elbow_in_base_r, loc=<-0.0001 0 0.0002>, rot=<0 0 0 1>}
 deform_elbow_out_base_r {parent=lowerarm_r, loc=<0 4 -0.0001>, rot=<0 0 0 -1>}
 deform_elbow_out_r {parent=deform_elbow_out_base_r, loc=<0 0 0>, rot=<0 0 0 1>}
 upperarm_twist_01_r {parent=upperarm_r, loc=<-0.0041 0 0>, rot=<0 0 0 -1>}
 deform_upperarm_bk_base_r {parent=upperarm_twist_01_r, loc=<-1.0814 7.2192 1.9455>, rot=<0 0 0 1>}
 deform_upperarm_bk_r {parent=deform_upperarm_bk_base_r, loc=<0 0 0>, rot=<0 0 0 1>}
 deform_upperarm_fr_base_r {parent=upperarm_twist_01_r, loc=<0.3689 -7.0506 2.7652>, rot=<0 0 0 1>}
 deform_upperarm_fr_r {parent=deform_upperarm_fr_base_r, loc=<0 0 0>, rot=<0 0 0 1>}
 deform_upperarm_up_base_r {parent=upperarm_twist_01_r, loc=<3.2214 1.1391 -3.3308>, rot=<0 0 0 1>}
 deform_upperarm_up_r {parent=deform_upperarm_up_base_r, loc=<0 0 0>, rot=<0 0 0 -1>}
 upperarm_twist_02_r {parent=upperarm_r, loc=<-13.7941 0 0>, rot=<0 0 0 -1>}
 dyn_jacket_1_base {parent=spine_03, loc=<3.9801 12.1634 -6.5775>, rot=<0.061 -0.9975 -0.0017 -0.0343>}
 dyn_jacket_1_1 {parent=dyn_jacket_1_base, loc=<0 0 0>, rot=<0 0 0 1>}
 dyn_jacket_1_2 {parent=dyn_jacket_1_1, loc=<16.1631 0.0404 -0.2184>, rot=<0 0 0 1>}
 dyn_jacket_2_base {parent=spine_03, loc=<1.2031 -0.6053 -17.3399>, rot=<0.0824 -0.9964 0.0005 0.0179>}
 dyn_jacket_2_1 {parent=dyn_jacket_2_base, loc=<0 0 0>, rot=<0 0 0 1>}
 dyn_jacket_2_2 {parent=dyn_jacket_2_1, loc=<12.1369 0 0>, rot=<0 0 0 1>}
 dyn_jacket_3_base {parent=spine_03, loc=<3.4521 -11.4283 -0.1517>, rot=<0.0947 -0.9955 0 0.0028>}
 dyn_jacket_3_1 {parent=dyn_jacket_3_base, loc=<0 0 0>, rot=<0 0 0 1>}
 dyn_jacket_3_2 {parent=dyn_jacket_3_1, loc=<11.0718 0 0>, rot=<0 0 0 1>}
 dyn_jacket_4_base {parent=spine_03, loc=<1.4444 -0.4028 17.0356>, rot=<0.0773 -0.997 -0.0002 -0.0071>}
 dyn_jacket_4_1 {parent=dyn_jacket_4_base, loc=<0 0 0>, rot=<0 0 0 1>}
 dyn_jacket_4_2 {parent=dyn_jacket_4_1, loc=<12.1047 0 0>, rot=<0 0 0 1>}
 dyn_jacket_5_base {parent=spine_03, loc=<4.1712 11.6857 9.0224>, rot=<0.0541 -0.9985 0.0005 0.0091>}
 dyn_jacket_5_1 {parent=dyn_jacket_5_base, loc=<0 0 0>, rot=<0 0 0 1>}
 dyn_jacket_5_2 {parent=dyn_jacket_5_1, loc=<16.1463 -0.0043 0.0581>, rot=<0 0 0 1>}
 dyn_button_r {parent=spine_02, loc=<1.2937 12.2603 5.0566>, rot=<0.0282 -0.7065 -0.0282 -0.7065>}
 dyn_button_l {parent=spine_02, loc=<1.3592 12.1492 -4.0626>, rot=<0.0282 -0.7065 -0.0282 -0.7065>}
 thigh_l {parent=pelvis, loc=<-2.5693 -0.0109 -7.7501>, rot=<0.0012 0.0572 0.0155 -0.9982>}
 calf_l {parent=thigh_l, loc=<-39.0484 0 0>, rot=<0.0006 -0.0161 -0.0391 -0.9991>}
 foot_l {parent=calf_l, loc=<-42.561 0.1077 0.4292>, rot=<0.0418 -0.0455 0.0343 -0.9975>}
 ball_l {parent=foot_l, loc=<-6.1599 13.8264 0.6806>, rot=<0.0001 -0.0001 -0.7063 -0.7079>}
 calf_twist_02_l {parent=calf_l, loc=<-18.8064 -0.0967 0.0397>, rot=<0.0237 0.0095 -0.0035 -0.9997>}
 calf_twist_01_l {parent=calf_l, loc=<-35.813 -0.0967 0.0397>, rot=<0.0237 0.0095 -0.0035 -0.9997>}
 deform_knee_in_base_l {parent=calf_l, loc=<1.298 -5.2957 -0.0496>, rot=<0 0 0 -1>}
 deform_knee_in_l {parent=deform_knee_in_base_l, loc=<0 0 0>, rot=<0 0 0 1>}
 deform_knee_out_base_l {parent=calf_l, loc=<0 4 0>, rot=<0 0 0 -1>}
 deform_knee_out_l {parent=deform_knee_out_base_l, loc=<0 0 0>, rot=<0 0 0 1>}
 thigh_twist_01_l {parent=thigh_l, loc=<-21.6303 0 0>, rot=<0.0013 0 0 -1>}
 deform_groin_base_l {parent=thigh_l, loc=<-2.1908 8.1142 -1.7442>, rot=<0 0 0 -1>}
 deform_groin_l {parent=deform_groin_base_l, loc=<0 0 0>, rot=<0 0 0 1>}
 thigh_r {parent=pelvis, loc=<-2.5693 -0.0109 7.7501>, rot=<0.0572 -0.0012 0.9982 0.0155>}
 calf_r {parent=thigh_r, loc=<39.0484 0 0>, rot=<0.0006 -0.0161 -0.0391 -0.9991>}
 foot_r {parent=calf_r, loc=<42.561 -0.1077 -0.4292>, rot=<0.0418 -0.0455 0.0343 -0.9975>}
 ball_r {parent=foot_r, loc=<6.1599 -13.8264 -0.6806>, rot=<0.0001 -0.0001 -0.7063 -0.7079>}
 calf_twist_02_r {parent=calf_r, loc=<18.8064 0.0967 -0.0397>, rot=<0.0237 0.0095 -0.0035 -0.9997>}
 calf_twist_01_r {parent=calf_r, loc=<35.813 0.0967 -0.0397>, rot=<0.0237 0.0095 -0.0035 -0.9997>}
 deform_knee_in_base_r {parent=calf_r, loc=<-1.298 5.2957 0.0496>, rot=<0 0 0 -1>}
 deform_knee_in_r {parent=deform_knee_in_base_r, loc=<0 0 0>, rot=<0 0 0 1>}
 deform_knee_out_base_r {parent=calf_r, loc=<0 -4 0>, rot=<0 0 0 -1>}
 deform_knee_out_r {parent=deform_knee_out_base_r, loc=<0 0 0>, rot=<0 0 0 1>}
 thigh_twist_01_r {parent=thigh_r, loc=<21.6303 0 0>, rot=<0.0013 0 0 -1>}
 deform_groin_base_r {parent=thigh_r, loc=<2.1908 -8.1142 1.7442>, rot=<0 0 0 -1>}
 deform_groin_r {parent=deform_groin_base_r, loc=<0 0 0>, rot=<0 0 0 1>}
 deform_glute_base_l {parent=pelvis, loc=<-8.8064 -8.8222 -8.4287>, rot=<0.0275 0.0501 -0.4472 -0.8926>}
 deform_glute_l {parent=deform_glute_base_l, loc=<0 0 0>, rot=<0 0 0 1>}
 deform_glute_base_r {parent=pelvis, loc=<-8.8062 -8.8222 8.4287>, rot=<0.0501 -0.0275 0.8926 -0.4472>}
 deform_glute_r {parent=deform_glute_base_r, loc=<0 0 0>, rot=<0 0 0 1>}
]
variables [
 is_RigidBody_alpha_FN {type=double, default=0}
 is_crouch {type=double, default=0}
 is_emote {type=double, default=0}
 is_move_backpack {type=double, default=0}
 is_sky_diving {type=double, default=0}
]
