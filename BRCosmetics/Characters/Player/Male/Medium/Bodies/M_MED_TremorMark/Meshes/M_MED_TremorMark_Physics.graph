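# Physics asset for the TremorMark medium male player body, set up for secondary motion on five dynamic jacket bone chains.
# Each chain links a base and two segments with angular limits, mostly hinge-like with wider swing1 than swing2 or twist.
# Capsules on pelvis, lower spine and thighs provide body collision volumes; the lower segments of chains 1 and 5 carry extra capsules.
# An angularly free constraint cross-links the lower segments of chains 1 and 2.
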PhysicsAsset /BRCosmetics/Characters/Player/Male/Medium/Bodies/M_MED_TremorMark/Meshes/M_MED_TremorMark_Physics
bodies {
 dyn_jacket_1_1 [{
   shapeType=Capsule
   length=12.2153
   radius=5.842
   shapeName=dyn_jacket_1_1_capsule
   center=<12.2889 2.1982 -2.7837>
   rotation=<rot 65.4385 -121.1338 -109.5224>
  }]
 dyn_jacket_1_2 [
  {
   shapeType=Capsule
   length=10.7033
   radius=5.3636
   shapeName=dyn_jacket_1_2_capsule
   center=<6.5935 2.1611 -0.0787>
   rotation=<rot -63.4928 70.9655 -68.0854>
  }
  {
   shapeType=Capsule
   length=14
   radius=5.75
   shapeName=dyn_jacket_1_2_capsule
   center=<10.5541 -3.6613 9.3591>
   rotation=<rot 2.1109 -4.5313 -35.0837>
  }
 ]
 dyn_jacket_1_base [{shapeType=Capsule, length=0.505, radius=1.755, shapeName=dyn_jacket_1_base_capsule, rotation=<rot 0 -90 0>}]
 dyn_jacket_2_1 [{
   shapeType=Capsule
   length=17.6746
   radius=5.8347
   shapeName=dyn_jacket_2_1_capsule
   center=<11.6292 -0.4599 2.5483>
   rotation=<rot -0.1348 0.0132 80.1513>
  }]
 dyn_jacket_2_2 [{
   shapeType=Capsule
   length=19.5527
   radius=5.8399
   shapeName=dyn_jacket_2_2_capsule
   center=<12.3074 -2.8819 2.0925>
   rotation=<rot 1.4061 1.0093 71.9983>
  }]
 dyn_jacket_2_base [{shapeType=Capsule, length=0.505, radius=1.755, shapeName=dyn_jacket_2_base_capsule, rotation=<rot 0 -90 0>}]
 dyn_jacket_3_1 [{
   shapeType=Capsule
   length=11.0788
   radius=9.3509
   shapeName=dyn_jacket_3_1_capsule
   center=<15.0194 -5.919 0.2883>
   rotation=<rot 2.6883 -0.0591 -1.2705>
  }]
 dyn_jacket_3_2 [{
   shapeType=Capsule
   length=15.4172
   radius=10.6375
   shapeName=dyn_jacket_3_2_capsule
   center=<16.6363 -5.1432 0.3146>
   rotation=<rot 3.0788 0.026 0.4714>
  }]
 dyn_jacket_3_base [{shapeType=Capsule, length=0.505, radius=1.755, shapeName=dyn_jacket_3_base_capsule, rotation=<rot 0 -90 0>}]
 dyn_jacket_4_1 [{
   shapeType=Capsule
   length=9.9574
   radius=4.8628
   shapeName=dyn_jacket_4_1_capsule
   center=<12.1163 0.4237 -0.0922>
   rotation=<rot -0.5347 6.1237 -78.0341>
  }]
 dyn_jacket_4_2 [{
   shapeType=Capsule
   length=17.2874
   radius=5.6189
   shapeName=dyn_jacket_4_2_capsule
   center=<11.8071 -0.3316 -2.4367>
   rotation=<rot -4.4893 1.1929 -82.5425>
  }]
 dyn_jacket_4_base [{shapeType=Capsule, length=0.505, radius=1.755, shapeName=dyn_jacket_4_base_capsule, rotation=<rot 0 -90 0>}]
 dyn_jacket_5_1 [{
   shapeType=Capsule
   length=10.7345
   radius=5.0144
   shapeName=dyn_jacket_5_1_capsule
   center=<13.0711 3.0347 1.3243>
   rotation=<rot -60.1532 -88.5575 87.703>
  }]
 dyn_jacket_5_2 [
  {
   shapeType=Capsule
   length=7.0059
   radius=5.4339
   shapeName=dyn_jacket_5_2_capsule
   center=<8.3757 3.4498 0.4214>
   rotation=<rot 68.9356 84.7211 90.4034>
  }
  {shapeType=Capsule, length=16.25, radius=5.25, shapeName=dyn_jacket_5_2_capsule, center=<12.0111 -6.4233 -10.0043>, rotation=<rot 0 0.001 50>}
 ]
 dyn_jacket_5_base [{shapeType=Capsule, length=0.505, radius=1.755, shapeName=dyn_jacket_5_base_capsule, rotation=<rot 0 -90 0>}]
 pelvis [{shapeType=Capsule, length=9.8789, radius=11.2312, shapeName=pelvis_capsule, center=<-3.6675 -1.446 0.0866>, rotation=<rot 1.3461 -90 0.6788>}]
 spine_01 [{
   shapeType=Capsule
   length=8.001
   radius=10.3246
   shapeName=spine_01_capsule
   center=<1.9681 -0.1012 -0.927>
   rotation=<rot 5.6567 -90 -0.3613>
  }]
 spine_02 [{shapeType=Capsule, length=9.388, radius=9.4356, shapeName=spine_02_capsule, center=<2.8681 -0.4041 -0.651>, rotation=<rot 0.2013 -90 0.6765>}]
 thigh_l [{
   shapeType=Capsule
   length=31.1654
   radius=8.2815
   shapeName=thigh_l_capsule
   center=<-16.6778 -1.1328 0.871>
   rotation=<rot 0.1799 -92.5755 -85.8108>
  }]
 thigh_r [{
   shapeType=Capsule
   length=31.4548
   radius=8.0292
   shapeName=thigh_r_capsule
   center=<16.4713 1.3452 -0.6691>
   rotation=<rot 0.2376 -94.5417 -87.2838>
  }]
}
constraints [
 {bone1Name=dyn_jacket_1_base, bone2Name=dyn_jacket_1_1, swing1LimitDegrees=40, swing2LimitDegrees=5, twistLimitDegrees=10}
 {bone1Name=dyn_jacket_2_base, bone2Name=dyn_jacket_2_1, swing1LimitDegrees=30, swing2LimitDegrees=5, twistLimitDegrees=2}
 {bone1Name=dyn_jacket_2_2, bone2Name=dyn_jacket_2_1, swing1LimitDegrees=30, swing2LimitDegrees=5, twistLimitDegrees=2}
 {bone1Name=dyn_jacket_3_base, bone2Name=dyn_jacket_3_1, swing1LimitDegrees=20, swing2LimitDegrees=5, twistLimitDegrees=2}
 {bone1Name=dyn_jacket_3_2, bone2Name=dyn_jacket_3_1, swing1LimitDegrees=30, swing2LimitDegrees=5, twistLimitDegrees=2}
 {bone1Name=dyn_jacket_4_base, bone2Name=dyn_jacket_4_1, swing1LimitDegrees=30, swing2LimitDegrees=5, twistLimitDegrees=2}
 {bone1Name=dyn_jacket_4_2, bone2Name=dyn_jacket_4_1, swing1LimitDegrees=30, swing2LimitDegrees=5, twistLimitDegrees=2}
 {bone1Name=dyn_jacket_5_base, bone2Name=dyn_jacket_5_1, swing1LimitDegrees=30, swing2LimitDegrees=5, twistLimitDegrees=2}
 {bone1Name=dyn_jacket_1_2, bone2Name=dyn_jacket_1_1, swing1LimitDegrees=30, swing2LimitDegrees=10, twistLimitDegrees=15}
 {bone1Name=dyn_jacket_5_2, bone2Name=dyn_jacket_5_1, swing1LimitDegrees=45, swing2LimitDegrees=5, twistLimitDegrees=2}
 {
  bone1Name=dyn_jacket_2_2
  bone2Name=dyn_jacket_1_2
  swing1LimitDegrees=45
  swing1Motion=Free
  swing2LimitDegrees=45
  swing2Motion=Free
  twistLimitDegrees=45
  twistMotion=Free
 }
]
