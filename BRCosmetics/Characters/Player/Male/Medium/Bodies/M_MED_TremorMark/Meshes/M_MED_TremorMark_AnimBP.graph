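# Post-copy AnimBP for the TremorMark medium male body, adding secondary motion to the jacket and buttons.
# The copied parent pose runs through a Control Rig, then two spine_02-space RigidBody sims, toggled between game and front-end physics assets.
# Constraints pin jacket bones to spine_01 when skydiving with back bling and blend both buttons across weighted bones.
# Each update derives rig inputs from crouch, emote and skydive states, plus gravity via a deprecated override function.
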AnimBlueprint /BRCosmetics/Characters/Player/Male/Medium/Bodies/M_MED_TremorMark/Meshes/M_MED_TremorMark_AnimBP
target_skeleton=&/M_MED_TremorMark_Skeleton
graphs [
 AnimGraph {nodes [
   n1 = Control Rig!(
    Source=Copy Pose From Mesh() {bUseAttachedParent=true}
    Alpha
    is_crouch:double=Get Dyn_is_crouch() -> (Dyn_is_crouch:double)
    is_emote:double=Get Dyn_Is_emote() -> (Dyn_Is_emote:double)
    is_sky_diving:double=Get Dyn_Is_sky_diving() -> (Dyn_Is_sky_diving:double)
    is_move_backpack:double=Get Is Move Backpack() -> (Is Move Backpack:double)
    is_RigidBody_alpha_FN:double=Get Dyn_RigidBody_Alpha_FN() -> (Dyn_RigidBody_Alpha_FN:float)
   ) -> (Pose) {controlRigAssetReference {blueprintRigClass=&/M_MED_TremorMark_CtrlRig.M_MED_TremorMark_CtrlRig_C}}
   n2 = Get Gravity Override()
   n3 = RigidBody!(
    ComponentPose=Local To Component(LocalPose=@n1)
    Alpha=Get Dyn_RigidBody_Alpha() -> (Dyn_RigidBody_Alpha:float)
    OverrideWorldGravity:Vector=@n2
    ExternalForce
   ) -> (Pose) {
    bOverrideWorldGravity=true
    overridePhysicsAsset=&/M_MED_TremorMark_Physics
    simulationSpace=BaseBoneSpace
    baseBoneRef {boneName=spine_02}
    componentAppliedLinearAccClamp=<3000 3000 3000>
    overrideWorldGravity {z=-980}
    simSpaceSettings {maxAngularAcceleration=50, maxAngularVelocity=5, maxLinearAcceleration=600, maxLinearVelocity=120, worldAlpha=1, externalLinearDragV=<1 1 1>}
   }
   n4 = RigidBody(ComponentPose=@n3, Alpha=Get Dyn_RigidBody_Alpha_FN(), OverrideWorldGravity=@n2) {
    bOverrideWorldGravity=true
    overridePhysicsAsset=&/M_MED_TremorMark_Physics_FE
    simulationSpace=BaseBoneSpace
    baseBoneRef {boneName=spine_02}
    componentAppliedLinearAccClamp=<3000 3000 3000>
    overrideWorldGravity {z=-980}
    simSpaceSettings {maxAngularAcceleration=50, maxAngularVelocity=5, maxLinearAcceleration=600, maxLinearVelocity=120, worldAlpha=1, externalLinearDragV=<1 1 1>}
   }
   n5 = Get Dyn_Is_sky_diving_1_backpack() -> (Dyn_Is_sky_diving_1_backpack:double)
   n6 = "Constraint - dyn_jacket_2_1 (P)"(ComponentPose=@n4, Alpha=@n5, ConstraintWeights_0=1) {
    alphaScaleBiasClamp {bInterpResult=true, interpSpeedDecreasing=100, interpSpeedIncreasing=100}
    boneToModify {boneName=dyn_jacket_2_1}
    constraintSetup [{targetBone {boneName=spine_01}}]
    constraintWeights [1]
   }
   n7 = "Constraint - dyn_jacket_4_1 (P)"(ComponentPose=@n6, Alpha=@n5, ConstraintWeights_0=1) {
    alphaScaleBiasClamp {bInterpResult=true, interpSpeedDecreasing=100, interpSpeedIncreasing=100}
    boneToModify {boneName=dyn_jacket_4_1}
    constraintSetup [{targetBone {boneName=spine_01}}]
    constraintWeights [1]
   }
   n8 = Get Dyn_Is_sky_diving_2_backpack() -> (Dyn_Is_sky_diving_2_backpack:double)
   n9 = "Constraint - dyn_jacket_3_1 (P)"(ComponentPose=@n7, Alpha=@n8, ConstraintWeights_0=1) {
    alphaScaleBiasClamp {bInterpResult=true, interpSpeedDecreasing=100, interpSpeedIncreasing=100}
    boneToModify {boneName=dyn_jacket_3_1}
    constraintSetup [{targetBone {boneName=spine_01}}]
    constraintWeights [1]
   }
   n10 = "Constraint - dyn_jacket_3_2 (P)"(ComponentPose=@n9, Alpha=@n8, ConstraintWeights_0=1) {
    alphaScaleBiasClamp {bInterpResult=true, interpSpeedDecreasing=100, interpSpeedIncreasing=100}
    boneToModify {boneName=dyn_jacket_3_2}
    constraintSetup [{targetBone {boneName=spine_01}}]
    constraintWeights [1]
   }
   n11 = "Constraint - dyn_button_l (PPPPPPPP)"!(
    ComponentPose=@n10
    Alpha=1
    ConstraintWeights_0=0.4456
    ConstraintWeights_1:float=0.1978
    ConstraintWeights_2:float=0.151
    ConstraintWeights_3:float=0.1185
    ConstraintWeights_4:float=0.0908
    ConstraintWeights_5:float=0.0058
    ConstraintWeights_6:float=0.0034
    ConstraintWeights_7:float=0.0023
   ) -> (Pose) {
    boneToModify {boneName=dyn_button_l}
    constraintSetup [
     {targetBone {boneName=spine_03}}
     {targetBone {boneName=spine_02}}
     {targetBone {boneName=dyn_jacket_1_1}}
     {targetBone {boneName=dyn_jacket_1_2}}
     {targetBone {boneName=spine_04}}
     {targetBone {boneName=dyn_jacket_2_2}}
     {targetBone {boneName=clavicle_l}}
     {targetBone {boneName=dyn_jacket_2_1}}
    ]
    constraintWeights [1,1,1,1,1,1,1,1]
   }
   n12 = "Constraint - dyn_button_r (PPPPPPPP)"!(
    ComponentPose=@n11
    Alpha=1
    ConstraintWeights_0=0.62
    ConstraintWeights_1:float=0.1724
    ConstraintWeights_2:float=0.1035
    ConstraintWeights_3:float=0.0844
    ConstraintWeights_4:float=0.0676
    ConstraintWeights_5:float=0.0207
    ConstraintWeights_6:float=0.0177
    ConstraintWeights_7:float=0.0047
   ) -> (Pose) {
    boneToModify {boneName=dyn_button_r}
    constraintSetup [
     {targetBone {boneName=dyn_jacket_5_1}}
     {targetBone {boneName=spine_03}}
     {targetBone {boneName=spine_02}}
     {targetBone {boneName=spine_04}}
     {targetBone {boneName=dyn_jacket_5_2}}
     {targetBone {boneName=spine_05}}
     {targetBone {boneName=clavicle_r}}
     {targetBone {boneName=dyn_jacket_4_1}}
    ]
    constraintWeights [1,1,1,1,1,1,1,1]
   }
   Output Pose(Result=Component To Local(ComponentPose=@n12)) {layerGroup=None}
  ]}
 EventGraph {nodes [
   Event BlueprintUpdateAnimation()
   > n1 = Sequence()
   > n2 = Branch(Condition=Get bIsInFrontEnd())
   > Set Dyn_RigidBody_Alpha(Dyn_RigidBody_Alpha:float=0) -> (Output_Get:float)
   > Set Dyn_RigidBody_Alpha_FN(Dyn_RigidBody_Alpha_FN:float=1) -> (Output_Get:float)
   Set Dyn_RigidBody_Alpha(execute=@n2.else, Dyn_RigidBody_Alpha=1)
   > Set Dyn_RigidBody_Alpha_FN(Dyn_RigidBody_Alpha_FN=0)
   n3 = Sequence!(execute=@n1.then_1) -> (then_0, then_1, then_2:Exec)
   > n4 = Gravity Override State Function DEPRECATED(
    CharacterPartAnimInstance=Self-Reference()
    GameTimeInSeconds=GetGameTimeInSeconds()
    GravityOverrideParams=Get Gravity Override Params()
    RBAN Params=Get RBAN Params()
   )
   > Set Gravity Override(Gravity Override=@n4)
   n5 = Sequence(execute=@n3.then_1)
   > n6 = Branch(Condition=Get bIsCrouching())
   > Set Dyn_is_crouch(Dyn_is_crouch:double=1) -> (Output_Get:double)
   Set Dyn_is_crouch(execute=@n6.else, Dyn_is_crouch=0)
   n7 = Branch(execute=@n5.then_1, Condition=Get bIsPlayingEmote())
   > Set Dyn_Is_emote(Dyn_Is_emote:double=0.33) -> (Output_Get:double)
   Set Dyn_Is_emote(execute=@n7.else, Dyn_Is_emote=0)
   n8 = Branch(execute=@n5.then_2, Condition=Get bIsSkyDiving())
   > n9 = Branch(Condition=Get bIsParachuteOpened())
   > Set Dyn_Is_sky_diving(Dyn_Is_sky_diving:double=1) -> (Output_Get:double)
   Set Dyn_Is_sky_diving(execute=@n9.else, Dyn_Is_sky_diving=0.5)
   Set Dyn_Is_sky_diving(execute=@n8.else, Dyn_Is_sky_diving=0)
   n10 = CheckForBackbling()
   n11 = Set Is Wearing Backbling(execute=@n3.then_2, Is Wearing Backbling=OR Boolean(A=@n10, B=@n10.bIsBackblingACape))
   > n12 = Branch(Condition=AND Boolean(A=@n11, B=Get bIsSkyDiving()))
   > n13 = Branch(Condition=Get bIsParachuteOpened())
   > Set Dyn_Is_sky_diving_1_backpack(Dyn_Is_sky_diving_1_backpack:double=0.5) -> (Output_Get:double)
   > Set Dyn_Is_sky_diving_2_backpack(Dyn_Is_sky_diving_2_backpack:double=0.5) -> (Output_Get:double)
   > Set Is Move Backpack(Is Move Backpack:double=0.25) -> (Output_Get:double)
   n14 = Branch(execute=@n13.else, Condition=Get bIsSkydiveFloating())
   > Set Dyn_Is_sky_diving_1_backpack(Dyn_Is_sky_diving_1_backpack=0.5)
   > Set Dyn_Is_sky_diving_2_backpack(Dyn_Is_sky_diving_2_backpack=1)
   > Set Is Move Backpack(Is Move Backpack=1)
   Set Dyn_Is_sky_diving_1_backpack(execute=@n14.else, Dyn_Is_sky_diving_1_backpack=0.5)
   > Set Dyn_Is_sky_diving_2_backpack(Dyn_Is_sky_diving_2_backpack=1)
   > Set Is Move Backpack(Is Move Backpack=0)
   Set Dyn_Is_sky_diving_1_backpack(execute=@n12.else, Dyn_Is_sky_diving_1_backpack=0)
   > Set Dyn_Is_sky_diving_2_backpack(Dyn_Is_sky_diving_2_backpack=0)
   > Set Is Move Backpack(Is Move Backpack=0)
  ]}
]
variables [
 Dyn_is_crouch {type=number}
 Dyn_Is_emote {type=number, default=1}
 Dyn_Is_sky_diving {type=number, default=1}
 Dyn_Is_sky_diving_1_backpack {type=number, default=1}
 Dyn_Is_sky_diving_2_backpack {type=number, default=1}
 Dyn_RigidBody_Alpha {type=number, default=1}
 Dyn_RigidBody_Alpha_FN {type=number, default=1}
 Gravity Override {type=Vector}
 Gravity Override Params {
  type=GravityOverrideParamsStruct
  default {
   skydive_DivingDown {gravityOverride=<0 0 0>}
   skydive_FloatingDown {gravityOverride=<0 -200 450>}
   skydive_GoingUp {gravityOverride=<0 0 -300>}
   skydive_Parachuting {gravityOverride=<0 -600 -980>}
  }
 }
 Is Move Backpack {type=number}
 Is Wearing Backbling {type=boolean}
 RBAN Params {
  type=RBANParamsStruct
  default {
   dBNO {simSpaceSettings {
     maxAngularAcceleration=1000
     maxAngularVelocity=3
     maxLinearAcceleration=300
     maxLinearVelocity=700
     velocityScaleZ=0
     externalLinearDragV=<0 0.5 0>
    }}
   dBNO_Carried {simSpaceSettings {
     maxAngularAcceleration=1000
     maxAngularVelocity=3
     maxLinearAcceleration=300
     maxLinearVelocity=700
     velocityScaleZ=0
     externalLinearDragV=<0 0.5 0>
    }}
   emoteOrMelee {simSpaceSettings {
     maxAngularAcceleration=1000
     maxAngularVelocity=3
     maxLinearAcceleration=300
     maxLinearVelocity=700
     velocityScaleZ=0
     externalLinearDragV=<0 0.5 0>
    }}
   front_End {simSpaceSettings {
     maxAngularAcceleration=1000
     maxAngularVelocity=3
     maxLinearAcceleration=300
     maxLinearVelocity=700
     velocityScaleZ=0
     externalLinearDragV=<0 0.5 0>
    }}
   onGround_Crouch_Moving {simSpaceSettings {
     maxAngularAcceleration=1000
     maxAngularVelocity=3
     maxLinearAcceleration=300
     maxLinearVelocity=700
     velocityScaleZ=0
     externalLinearDragV=<0 0.5 0>
    }}
   onGround_Falling {simSpaceSettings {
     maxAngularAcceleration=1000
     maxAngularVelocity=3
     maxLinearAcceleration=300
     maxLinearVelocity=700
     velocityScaleZ=0
     externalLinearDragV=<0 0.5 0>
    }}
   onGround_Moving {simSpaceSettings {
     maxAngularAcceleration=1000
     maxAngularVelocity=3
     maxLinearAcceleration=300
     maxLinearVelocity=700
     velocityScaleZ=0
     externalLinearDragV=<0 0.5 0>
    }}
   onGround_Standing {simSpaceSettings {
     maxAngularAcceleration=1000
     maxAngularVelocity=3
     maxLinearAcceleration=300
     maxLinearVelocity=700
     velocityScaleZ=0
     externalLinearDragV=<0 0.5 0>
    }}
   rail_Grinding {simSpaceSettings {
     maxAngularAcceleration=1000
     maxAngularVelocity=3
     maxLinearAcceleration=300
     maxLinearVelocity=700
     velocityScaleZ=0
     worldAlpha=0.1
     externalLinearDragV=<0 0.5 0>
    }}
   riding_Moving {simSpaceSettings {
     maxAngularAcceleration=1000
     maxAngularVelocity=3
     maxLinearAcceleration=300
     maxLinearVelocity=700
     velocityScaleZ=0
     externalLinearDragV=<0 0.5 0>
    }}
   riding_Moving_Targeting {simSpaceSettings {
     maxAngularAcceleration=1000
     maxAngularVelocity=3
     maxLinearAcceleration=300
     maxLinearVelocity=700
     velocityScaleZ=0
     externalLinearDragV=<0 0.5 0>
    }}
   riding_Standing {simSpaceSettings {
     maxAngularAcceleration=1000
     maxAngularVelocity=3
     maxLinearAcceleration=300
     maxLinearVelocity=700
     velocityScaleZ=0
     externalLinearDragV=<0 0.5 0>
    }}
   skydive_DivingDown {simSpaceSettings {
     maxAngularAcceleration=1000
     maxAngularVelocity=3
     maxLinearAcceleration=300
     maxLinearVelocity=700
     velocityScaleZ=0
     externalLinearDragV=<0 0.5 0>
    }}
   skydive_FloatingDown {
    componentLinearAccScale=<0 0.5 0>
    componentLinearVelScale=<0 0.5 0>
    simSpaceSettings {
     maxAngularAcceleration=1000
     maxAngularVelocity=3
     maxLinearAcceleration=300
     maxLinearVelocity=700
     velocityScaleZ=0
     externalLinearDragV=<0 0.5 0>
    }
   }
   skydive_GoingUp {simSpaceSettings {
     maxAngularAcceleration=1000
     maxAngularVelocity=3
     maxLinearAcceleration=300
     maxLinearVelocity=700
     velocityScaleZ=0
     externalLinearDragV=<0 0.5 0>
    }}
   skydive_Parachuting {
    componentLinearAccScale=<0 0.1 0.3>
    componentLinearVelScale=<0 0.1 0.3>
    simSpaceSettings {
     maxAngularAcceleration=1000
     maxAngularVelocity=3
     maxLinearAcceleration=300
     maxLinearVelocity=700
     velocityScaleZ=0
     externalLinearDragV=<0 0.5 0>
    }
   }
   skydive_Surfing {simSpaceSettings {
     maxAngularAcceleration=1000
     maxAngularVelocity=3
     maxLinearAcceleration=300
     maxLinearVelocity=700
     velocityScaleZ=0
     externalLinearDragV=<0 0.5 0>
    }}
   swimming {simSpaceSettings {
     maxAngularAcceleration=1000
     maxAngularVelocity=3
     maxLinearAcceleration=300
     maxLinearVelocity=700
     velocityScaleZ=0
     externalLinearDragV=<0 0.5 0>
    }}
   tactical_Sprint {simSpaceSettings {
     maxAngularAcceleration=1000
     maxAngularVelocity=3
     maxLinearAcceleration=300
     maxLinearVelocity=700
     velocityScaleZ=0
     externalLinearDragV=<0 0.5 0>
    }}
   ziplining {simSpaceSettings {
     maxAngularAcceleration=1000
     maxAngularVelocity=3
     maxLinearAcceleration=300
     maxLinearVelocity=700
     velocityScaleZ=0
     externalLinearDragV=<0 0.5 0>
    }}
  }
 }
]
