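# Physics asset for the face accessory part of the TremorMark male medium character cosmetic, defining bodies for its bangs and back hair.
# Each of the two bangs and the left and right back hair pieces has a capsule constrained to its own base capsule, with tighter swing and twist limits on the back hair.
# Larger head, neck and upper-spine capsules have no constraints and appear to serve as collision volumes for the hair.
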PhysicsAsset /BRCosmetics/Characters/Player/Male/Medium/Bodies/M_MED_TremorMark/Meshes/Parts/M_MED_TremorMark_FaceAcc_Physics
bodies {
 dyn_bangs_1 [{
   shapeType=Capsule
   length=3.8175
   radius=2.6006
   shapeName=dyn_bangs_1_capsule
   center=<4.5487 3.334 2.2502>
   rotation=<rot 22.6177 -89.6365 1.3897>
  }]
 dyn_bangs_1_base [{shapeType=Capsule, length=0.505, radius=1.505, shapeName=dyn_bangs_1_base_capsule, rotation=<rot 0 -90 0>}]
 dyn_bangs_2 [{
   shapeType=Capsule
   length=2.5189
   radius=1.9902
   shapeName=dyn_bangs_2_capsule
   center=<2.3425 1.6194 -0.2364>
   rotation=<rot 0 12.7693 -158.5064>
  }]
 dyn_bangs_2_base [{shapeType=Capsule, length=0.505, radius=1.505, shapeName=dyn_bangs_2_base_capsule, rotation=<rot 0 -90 0>}]
 dyn_hair_back_1_l [{shapeType=Capsule, length=5, radius=2.25, shapeName=dyn_hair_back_1_l_capsule, center=<5.318 0 0>, rotation=<rot 0 0.0005 0>}]
 dyn_hair_back_1_r [{shapeType=Capsule, length=5, radius=2.25, shapeName=dyn_hair_back_1_r_capsule, center=<5.318 -0.5357 0>}]
 dyn_hair_back_base_l [{shapeType=Capsule, length=0, radius=1.75, shapeName=dyn_hair_back_base_l_capsule, rotation=<rot 0 -90 0>}]
 dyn_hair_back_base_r [{shapeType=Capsule, length=0, radius=1.75, shapeName=dyn_hair_back_base_r_capsule, rotation=<rot 0 -90 0>}]
 head [{
   shapeType=Capsule
   length=2.7662
   radius=10.6856
   shapeName=head_capsule
   center=<1.6515 -0.7046 0.7489>
   rotation=<rot 14.5659 91.5196 -90.9285>
  }]
 neck_01 [{shapeType=Capsule, length=7, radius=5.5, shapeName=neck_01_capsule, center=<3.7063 0 0>, rotation=<rot 0 -100.0003 90>}]
 spine_05 [{shapeType=Capsule, length=10.005, radius=9.255, shapeName=spine_05_capsule, center=<5.6115 0.2258 0>, rotation=<rot 0 -90.001 0>}]
}
constraints [
 {bone1Name=dyn_bangs_1, bone2Name=dyn_bangs_1_base, swing1LimitDegrees=45, swing2LimitDegrees=45, twistLimitDegrees=30}
 {bone1Name=dyn_bangs_2, bone2Name=dyn_bangs_2_base, swing1LimitDegrees=45, swing2LimitDegrees=45, twistLimitDegrees=30}
 {bone1Name=dyn_hair_back_1_l, bone2Name=dyn_hair_back_base_l, swing1LimitDegrees=30, swing2LimitDegrees=30, twistLimitDegrees=20}
 {bone1Name=dyn_hair_back_1_r, bone2Name=dyn_hair_back_base_r, swing1LimitDegrees=30, swing2LimitDegrees=30, twistLimitDegrees=20}
]
